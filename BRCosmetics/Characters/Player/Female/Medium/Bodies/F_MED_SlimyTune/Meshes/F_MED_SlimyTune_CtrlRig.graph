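# Control Rig for the SlimyTune female medium cosmetic body, applying corrective deformation to body and clothing bones.
# A sequence runs a debug-only control-to-bone pass, then the shared DeformRig function followed by shoe band, jacket and strap fixes.
# Foot swing rotation drives shoe bands; thigh spherical pose readers and curve-remapped DeformRig attributes drive jacket and strap offsets.
# Movement-state booleans like crouching, sliding and skydiving blend extra offsets or pin jacket bones to the pelvis.
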ControlRigBlueprint /BRCosmetics/Characters/Player/Female/Medium/Bodies/F_MED_SlimyTune/Meshes/F_MED_SlimyTune_CtrlRig
graphs [
 New Function_ContainedGraph {nodes [
   n1 = RigVMFunctionEntryNode()
   RigVMFunction_ControlFlowBranch = &RigVMFunction_ControlFlowBranch(ExecuteContext=@n1.ExecuteContext, Condition=@n1.debug)
   RigVMFunctionReturnNode(ExecuteContext=@RigVMFunction_ControlFlowBranch.Completed)
   RigUnit_GetTransform = &RigUnit_GetTransform(Item=(Type=Control,Name="foot_l_ctrl"), bInitial=False)
   RigUnit_SetTransform = &RigUnit_SetTransform(
    Item=(Type=Bone,Name="foot_l")
    Weight=1
    bInitial=False
    Value=@RigUnit_GetTransform.Transform
    ExecutePin=@RigVMFunction_ControlFlowBranch.True
   )
   > &RigUnit_SetTransform(
    Item=(Type=Bone,Name="foot_r")
    Weight=1
    bInitial=False
    Value=&RigUnit_GetTransform(Item=(Type=Control,Name="foot_r_ctrl"), bInitial=False).Transform
   )
   > &RigUnit_SetTransform(
    Item=(Type=Bone,Name="thigh_l")
    Weight=1
    bInitial=False
    Value=&RigUnit_GetTransform(Item=(Type=Control,Name="thigh_l_ctrl"), bInitial=False).Transform
   )
   > &RigUnit_SetTransform(
    Item=(Type=Bone,Name="thigh_r")
    Weight=1
    bInitial=False
    Value=&RigUnit_GetTransform(Item=(Type=Control,Name="thigh_r_ctrl"), bInitial=False).Transform
   )
   > &RigUnit_SetTransform(
    Item=(Type=Bone,Name="upperarm_l")
    Weight=1
    bInitial=False
    Value=&RigUnit_GetTransform(Item=(Type=Control,Name="upperarm_l_ctrl"), bInitial=False).Transform
   )
  ]}
 New Function_ContainedGraph {nodes [
   RigVMFunctionEntryNode()
   > &RigUnit_ParentConstraint(
    AdvancedSettings=(InterpolationType=Average,RotationOrderForFilter=XZY)
    Child=(Type=Bone,Name="foot_l_get_rot")
    Parents=((Item=(Type=Bone,Name="foot_l"),Weight=1))
    Weight=1
   )
   > &RigUnit_ParentConstraint(
    AdvancedSettings=(InterpolationType=Average,RotationOrderForFilter=XZY)
    Child=(Type=Bone,Name="foot_r_get_rot")
    Parents=((Item=(Type=Bone,Name="foot_r"),Weight=1))
    Weight=1
   )
   RigUnit_GetTransform = &RigUnit_GetTransform(Item=(Type=Bone,Name="foot_l_get_rot"), Space=LocalSpace, bInitial=false)
   RigVMFunction_MathQuaternionSwingTwist = &RigVMFunction_MathQuaternionSwingTwist(TwistAxis=(X=0,Y=1,Z=0), Input=@RigUnit_GetTransform.Transform.Rotation)
   RigVMFunction_MathQuaternionToEuler = &RigVMFunction_MathQuaternionToEuler(RotationOrder=ZYX, Value=@RigVMFunction_MathQuaternionSwingTwist.Swing)
   Remap = &RigVMFunction_MathDoubleRemap(SourceMaximum=-45, SourceMinimum=0, TargetMaximum=1, TargetMinimum=0, Value=@RigVMFunction_MathQuaternionToEuler.Result.Z)
   > &RigUnit_ModifyTransforms(
    ItemToModify=((Item=(Type=Bone,Name="dyn_shoeband_l"),Transform=(Rotation=(X=0,Y=0.1736,Z=0,W=0.9848),Translation=(X=0,Y=0,Z=0),Scale3D=(X=1,Y=1,Z=1))))
    WeightMinimum=0
    Weight=@Remap.Result
   )
   n1 = &RigVMFunction_MathQuaternionSwingTwist(
    TwistAxis=(X=0,Y=1,Z=0)
    Input=&RigUnit_GetTransform(Item=(Type=Bone,Name="foot_r_get_rot"), Space=LocalSpace, bInitial=false).Transform.Rotation
   )
   > &RigUnit_ModifyTransforms(
    ItemToModify=((Item=(Type=Bone,Name="dyn_shoeband_r"),Transform=(Rotation=(X=0,Y=0.1736,Z=0,W=0.9848),Translation=(X=0,Y=0,Z=0),Scale3D=(X=1,Y=1,Z=1))))
    WeightMinimum=0
    Weight=&RigVMFunction_MathDoubleRemap(
     SourceMaximum=-45
     SourceMinimum=0
     TargetMaximum=1
     TargetMinimum=0
     Value=&RigVMFunction_MathQuaternionToEuler(RotationOrder=ZYX, Value=@n1.Swing).Result.Z
    ).Result
   )
   > RigVMFunctionReturnNode()
  ]}
 New Function_ContainedGraph {nodes [
   RigVMFunctionEntryNode()
   > RigVMFunction_Sequence = RigVMAggregateNode()
   AlphaInterp = &RigVMFunction_AlphaInterp(
    InterpSpeedDecreasing=1
    InterpSpeedIncreasing=1
    Scale=1
    bClampResult=true
    bMapRange=true
    Value=&RigVMFunction_MathBoolToFloat(Value=RigVMVariableNode() {variable=isTargetStandingUp}.Value).Result
   )
   &RigUnit_ParentConstraint(
    AdvancedSettings=(InterpolationType=Average,RotationOrderForFilter=XZY)
    Child=(Type=Bone,Name="base_jacket_0_l")
    Parents=((Item=(Type=Bone,Name="pelvis"),Weight=1))
    Weight=1
    Parents.0.Weight=@AlphaInterp.Result
    ExecutePin=@RigVMFunction_Sequence.B
   )
   > &RigUnit_ParentConstraint(
    AdvancedSettings=(InterpolationType=Average,RotationOrderForFilter=XZY)
    Child=(Type=Bone,Name="base_jacket_0_r")
    Parents=((Item=(Type=Bone,Name="pelvis"),Weight=1))
    Weight=1
    Parents.0.Weight=@AlphaInterp.Result
   )
   > &RigUnit_ParentConstraint(
    AdvancedSettings=(InterpolationType=Average,RotationOrderForFilter=XZY)
    Child=(Type=Bone,Name="base_jacket_1_l")
    Parents=((Item=(Type=Bone,Name="pelvis"),Weight=1))
    Weight=1
    Parents.0.Weight=@AlphaInterp.Result
   )
   > &RigUnit_ParentConstraint(
    AdvancedSettings=(InterpolationType=Average,RotationOrderForFilter=XZY)
    Child=(Type=Bone,Name="base_jacket_1_r")
    Parents=((Item=(Type=Bone,Name="pelvis"),Weight=1))
    Weight=1
    Parents.0.Weight=@AlphaInterp.Result
   )
   > &RigUnit_ParentConstraint(
    AdvancedSettings=(InterpolationType=Average,RotationOrderForFilter=XZY)
    Child=(Type=Bone,Name="base_jacket_2_l")
    Parents=((Item=(Type=Bone,Name="pelvis"),Weight=1))
    Weight=1
    Parents.0.Weight=@AlphaInterp.Result
   )
   > &RigUnit_ParentConstraint(
    AdvancedSettings=(InterpolationType=Average,RotationOrderForFilter=XZY)
    Child=(Type=Bone,Name="base_jacket_2_r")
    Parents=((Item=(Type=Bone,Name="pelvis"),Weight=1))
    Weight=1
    Parents.0.Weight=@AlphaInterp.Result
   )
   n1 = &RigVMFunction_MathBoolToFloat(Value=RigVMVariableNode() {variable=isCrouching}.Value)
   &RigUnit_ParentConstraint(
    AdvancedSettings=(InterpolationType=Average,RotationOrderForFilter=XZY)
    Child=(Type=Bone,Name="base_jacket_0_l")
    Parents=((Item=(Type=Bone,Name="pelvis"),Weight=1))
    Weight=1
    Parents.0.Weight=@n1.Result
    ExecutePin=@RigVMFunction_Sequence.C
   )
   > &RigUnit_ParentConstraint(
    AdvancedSettings=(InterpolationType=Average,RotationOrderForFilter=XZY)
    Child=(Type=Bone,Name="base_jacket_0_r")
    Parents=((Item=(Type=Bone,Name="pelvis"),Weight=1))
    Weight=1
    Parents.0.Weight=@n1.Result
   )
   > &RigUnit_ParentConstraint(
    AdvancedSettings=(InterpolationType=Average,RotationOrderForFilter=XZY)
    Child=(Type=Bone,Name="base_jacket_1_l")
    Parents=((Item=(Type=Bone,Name="pelvis"),Weight=1))
    Weight=1
    Parents.0.Weight=@n1.Result
   )
   > &RigUnit_ParentConstraint(
    AdvancedSettings=(InterpolationType=Average,RotationOrderForFilter=XZY)
    Child=(Type=Bone,Name="base_jacket_1_r")
    Parents=((Item=(Type=Bone,Name="pelvis"),Weight=1))
    Weight=1
    Parents.0.Weight=@n1.Result
   )
   > &RigUnit_ParentConstraint(
    AdvancedSettings=(InterpolationType=Average,RotationOrderForFilter=XZY)
    Child=(Type=Bone,Name="base_jacket_2_l")
    Parents=((Item=(Type=Bone,Name="pelvis"),Weight=1))
    Weight=1
    Parents.0.Weight=@n1.Result
   )
   > &RigUnit_ParentConstraint(
    AdvancedSettings=(InterpolationType=Average,RotationOrderForFilter=XZY)
    Child=(Type=Bone,Name="base_jacket_2_r")
    Parents=((Item=(Type=Bone,Name="pelvis"),Weight=1))
    Weight=1
    Parents.0.Weight=@n1.Result
   )
   n2 = &RigVMFunction_AnimEvalRichCurve(
    Curve="(EditorCurveData=(Keys=((InterpMode=RCIM_Cubic,TangentMode=RCTM_User,TangentWeightMode=RCTWM_WeightedNone,Time=0,Value=0,ArriveTangent=4.119,ArriveTangentWeight=0,LeaveTangent=4.119,LeaveTangentWeight=0),(InterpMode=RCIM_Cubic,TangentMode=RCTM_Auto,Tangent ...(truncated)"
    SourceMinimum=0
    TargetMinimum=0
    Value=RigVMDispatchNode(BoneName=root, Default=0, Name=deformRig_l_thigh_up_alpha) {template="DISPATCH_RigDispatch_GetAnimAttribute(in Name,in BoneName,in Default,out Value,out Found)"}.Value
   )
   &RigUnit_ModifyTransforms(
    ItemToModify="((Item=(Type=Bone,Name=\"base_jacket_0_l\"),Transform=(Rotation=(X=0.401,Y=-0.3181,Z=-0.4904,W=0.7054),Translation=(X=5.4042,Y=-1.4347,Z=3.6888),Scale3D=(X=1,Y=1,Z=1))),(Item=(Type=Bone,Name=\"dyn_jacket_0_end_l\"),Transform=(Rotation=(X=-0.1579,Y=0.2948,Z=0.3 ...(truncated)"
    WeightMinimum=0
    Weight=@n2.Result
    ExecutePin=@RigVMFunction_Sequence.E
   )
   > n3 = &RigUnit_SphericalPoseReader(
    ActiveRegionScaleFactors=(PositiveWidth=1,NegativeWidth=1,PositiveHeight=1,NegativeHeight=1)
    ActiveRegionSize=0.1
    Debug=(bDrawDebug=false,bDraw2D=False,bDrawLocalAxes=False,DebugScale=25,DebugSegments=20,DebugThickness=0.25)
    DriverAxis=(X=-1,Y=0,Z=0)
    DriverItem=(Type=Bone,Name="thigh_l")
    FalloffRegionScaleFactors=(PositiveWidth=1,NegativeWidth=2.8,PositiveHeight=1,NegativeHeight=2.8)
    FalloffSize=0.25
    FlipHeightScaling=False
    FlipWidthScaling=False
    OptionalParentItem=(Type=Bone,Name="None")
    RotationOffset=(X=180,Y=0,Z=0)
   )
   RigVMFunction_AnimEvalRichCurve = &RigVMFunction_AnimEvalRichCurve(
    Curve="(EditorCurveData=(Keys=((InterpMode=RCIM_Cubic,TangentMode=RCTM_User,TangentWeightMode=RCTWM_WeightedNone,Time=0,Value=0,ArriveTangent=0.4459,ArriveTangentWeight=0,LeaveTangent=0.4459,LeaveTangentWeight=0),(InterpMode=RCIM_Cubic,TangentMode=RCTM_User,Tange ...(truncated)"
    SourceMinimum=0
    TargetMinimum=0
    Value=@n3.OutputParam
   )
   > &RigUnit_ModifyTransforms(
    ItemToModify="((Item=(Type=Bone,Name=\"base_jacket_0_l\"),Transform=(Rotation=(X=0.0257,Y=0.0723,Z=-0.7214,W=0.6883),Translation=(X=13.0419,Y=-0.2057,Z=-0.6437),Scale3D=(X=1,Y=1,Z=1))),(Item=(Type=Bone,Name=\"base_jacket_1_l\"),Transform=(Rotation=(X=0.0565,Y=0.1365,Z=-0.69 ...(truncated)"
    WeightMinimum=0
    Weight=@RigVMFunction_AnimEvalRichCurve.Result
   )
   n4 = &RigVMFunction_AnimEvalRichCurve(
    Curve="(EditorCurveData=(Keys=((InterpMode=RCIM_Cubic,TangentMode=RCTM_User,TangentWeightMode=RCTWM_WeightedNone,Time=0,Value=0,ArriveTangent=4.119,ArriveTangentWeight=0,LeaveTangent=4.119,LeaveTangentWeight=0),(InterpMode=RCIM_Cubic,TangentMode=RCTM_Auto,Tangent ...(truncated)"
    SourceMinimum=0
    TargetMinimum=0
    Value=RigVMDispatchNode(BoneName=root, Default=0, Name=deformRig_r_thigh_up_alpha) {template="DISPATCH_RigDispatch_GetAnimAttribute(in Name,in BoneName,in Default,out Value,out Found)"}.Value
   )
   > &RigUnit_ModifyTransforms(
    ItemToModify="((Item=(Type=Bone,Name=\"base_jacket_0_r\"),Transform=(Rotation=(X=-0.3401,Y=0.3015,Z=-0.5682,W=0.686),Translation=(X=2.6697,Y=-5.6551,Z=-6.2715),Scale3D=(X=1,Y=1,Z=1))),(Item=(Type=Bone,Name=\"dyn_jacket_0_end_r\"),Transform=(Rotation=(X=0.1579,Y=-0.2948,Z=0. ...(truncated)"
    WeightMinimum=0
    Weight=@n4.Result
   )
   > n5 = &RigUnit_SphericalPoseReader(
    ActiveRegionScaleFactors=(PositiveWidth=1,NegativeWidth=1,PositiveHeight=1,NegativeHeight=1)
    ActiveRegionSize=0.1
    Debug=(bDrawDebug=false,bDraw2D=False,bDrawLocalAxes=False,DebugScale=25,DebugSegments=20,DebugThickness=0.25)
    DriverAxis=(X=-1,Y=0,Z=0)
    DriverItem=(Type=Bone,Name="thigh_r")
    FalloffRegionScaleFactors=(PositiveWidth=1,NegativeWidth=2.8,PositiveHeight=1,NegativeHeight=2.8)
    FalloffSize=0.25
    FlipHeightScaling=False
    FlipWidthScaling=False
    OptionalParentItem=(Type=Bone,Name="None")
    RotationOffset=(X=180,Y=0,Z=0)
   )
   n6 = &RigVMFunction_AnimEvalRichCurve(
    Curve="(EditorCurveData=(Keys=((InterpMode=RCIM_Cubic,TangentMode=RCTM_User,TangentWeightMode=RCTWM_WeightedNone,Time=0,Value=0,ArriveTangent=0.4459,ArriveTangentWeight=0,LeaveTangent=0.4459,LeaveTangentWeight=0),(InterpMode=RCIM_Cubic,TangentMode=RCTM_User,Tange ...(truncated)"
    SourceMinimum=0
    TargetMinimum=0
    Value=@n5.OutputParam
   )
   > &RigUnit_ModifyTransforms(
    ItemToModify="((Item=(Type=Bone,Name=\"base_jacket_0_r\"),Transform=(Rotation=(X=-0.0736,Y=-0.0219,Z=-0.7247,W=0.6848),Translation=(X=13.0419,Y=-0.2057,Z=-0.6437),Scale3D=(X=1,Y=1,Z=1))),(Item=(Type=Bone,Name=\"base_jacket_1_r\"),Transform=(Rotation=(X=0.0565,Y=0.1365,Z=-0. ...(truncated)"
    WeightMinimum=0
    Weight=@n6.Result
   )
   RigVMFunction_MathBoolToFloat = &RigVMFunction_MathBoolToFloat(Value=RigVMVariableNode() {variable=isCrouching}.Value)
   &RigUnit_ModifyTransforms(
    ItemToModify="((Item=(Type=Bone,Name=\"base_jacket_0_r\"),Transform=(Rotation=(X=-0.0031,Y=-0.116,Z=-0.1158,W=0.9865),Translation=(X=-1.6871,Y=1.9098,Z=6.5854),Scale3D=(X=1,Y=1,Z=1))),(Item=(Type=Bone,Name=\"base_jacket_0_l\"),Transform=(Rotation=(X=0.1798,Y=-0.0224,Z=-0.08 ...(truncated)"
    WeightMinimum=0
    Weight=@RigVMFunction_MathBoolToFloat.Result
    ExecutePin=@RigVMFunction_Sequence.G
   )
   > n7 = &RigUnit_SphericalPoseReader(
    ActiveRegionScaleFactors=(PositiveWidth=1,NegativeWidth=1,PositiveHeight=1,NegativeHeight=1)
    ActiveRegionSize=0.02
    Debug=(bDrawDebug=false,bDraw2D=False,bDrawLocalAxes=False,DebugScale=25,DebugSegments=20,DebugThickness=0.25)
    DriverAxis=(X=-1,Y=0,Z=0)
    DriverItem=(Type=Bone,Name="thigh_l")
    FalloffRegionScaleFactors=(PositiveWidth=1,NegativeWidth=1,PositiveHeight=1,NegativeHeight=1)
    FalloffSize=0.07
    FlipHeightScaling=False
    FlipWidthScaling=False
    OptionalParentItem=(Type=Bone,Name="None")
    RotationOffset=(X=115,Y=0,Z=-20)
   )
   > &RigUnit_ModifyTransforms(
    ItemToModify=((Item=(Type=Bone,Name="base_jacket_0_l"),Transform=(Rotation=(X=0.0547,Y=-0.2681,Z=-0.031,W=0.9613),Translation=(X=0,Y=0,Z=0),Scale3D=(X=1,Y=1,Z=1))))
    WeightMinimum=0
    Weight=@n7.OutputParam
   )
   > &RigUnit_ModifyTransforms(
    ItemToModify="((Item=(Type=Bone,Name=\"base_jacket_0_l\"),Transform=(Rotation=(X=-0.4169,Y=0.1121,Z=-0.0093,W=0.902),Translation=(X=1.3824,Y=4.9391,Z=-10.409),Scale3D=(X=1,Y=1,Z=1))),(Item=(Type=Bone,Name=\"base_jacket_0_r\"),Transform=(Rotation=(X=-0.1508,Y=0.1459,Z=-0.283 ...(truncated)"
    WeightMinimum=0
    Weight=&RigVMFunction_MathBoolToFloat(Value=RigVMVariableNode() {variable=isSliding}.Value).Result
   )
   > &RigUnit_ModifyTransforms(
    ItemToModify=((Item=(Type=Bone,Name="base_jacket_0_r"),Transform=(Rotation=(X=0.0564,Y=0.3815,Z=0.6661,W=0.6385),Translation=(X=4.791,Y=-0.631,Z=-3.4422),Scale3D=(X=1,Y=1,Z=1))))
    WeightMinimum=0
    Weight=&RigVMFunction_MathBoolToFloat(Value=RigVMVariableNode() {variable=isSlidingBaseBall}.Value).Result
   )
   > n8 = &RigUnit_SphericalPoseReader(
    ActiveRegionScaleFactors=(PositiveWidth=1,NegativeWidth=1,PositiveHeight=1,NegativeHeight=1)
    ActiveRegionSize=0.1
    Debug=(bDrawDebug=false,bDraw2D=False,bDrawLocalAxes=False,DebugScale=25,DebugSegments=20,DebugThickness=0.25)
    DriverAxis=(X=1,Y=0,Z=0)
    DriverItem=(Type=Bone,Name="thigh_r")
    FalloffRegionScaleFactors=(PositiveWidth=1,NegativeWidth=1,PositiveHeight=1,NegativeHeight=1)
    FalloffSize=0.2
    FlipHeightScaling=False
    FlipWidthScaling=False
    OptionalParentItem=(Type=Bone,Name="None")
    RotationOffset=(X=-90,Y=0,Z=-5)
   )
   > &RigUnit_ModifyTransforms(
    ItemToModify=((Item=(Type=Bone,Name="base_jacket_0_r"),Transform=(Rotation=(X=0,Y=0,Z=0,W=1),Translation=(X=4.9025,Y=0,Z=-3.3177),Scale3D=(X=1,Y=1,Z=1))))
    WeightMinimum=0
    Weight=@n8.OutputParam
   )
   n9 = &RigVMFunction_MathBoolToFloat(Value=RigVMVariableNode() {variable=isSkyDiving}.Value)
   &RigUnit_ParentConstraint(
    AdvancedSettings=(InterpolationType=Average,RotationOrderForFilter=XZY)
    Child=(Type=Bone,Name="base_jacket_0_l")
    Parents=((Item=(Type=Bone,Name="pelvis"),Weight=1))
    ExecutePin=@RigVMFunction_Sequence.H
    Weight=@n9.Result
   )
   > &RigUnit_ParentConstraint(
    AdvancedSettings=(InterpolationType=Average,RotationOrderForFilter=XZY)
    Child=(Type=Bone,Name="base_jacket_0_m")
    Parents=((Item=(Type=Bone,Name="pelvis"),Weight=1))
    Weight=@n9.Result
   )
   > &RigUnit_ParentConstraint(
    AdvancedSettings=(InterpolationType=Average,RotationOrderForFilter=XZY)
    Child=(Type=Bone,Name="base_jacket_0_r")
    Parents=((Item=(Type=Bone,Name="pelvis"),Weight=1))
    Weight=@n9.Result
   )
   > &RigUnit_ParentConstraint(
    AdvancedSettings=(InterpolationType=Average,RotationOrderForFilter=XZY)
    Child=(Type=Bone,Name="base_jacket_1_l")
    Parents=((Item=(Type=Bone,Name="pelvis"),Weight=1))
    Weight=@n9.Result
   )
   > &RigUnit_ParentConstraint(
    AdvancedSettings=(InterpolationType=Average,RotationOrderForFilter=XZY)
    Child=(Type=Bone,Name="base_jacket_1_l")
    Parents=((Item=(Type=Bone,Name="pelvis"),Weight=1))
    Weight=@n9.Result
   )
   > &RigUnit_ParentConstraint(
    AdvancedSettings=(InterpolationType=Average,RotationOrderForFilter=XZY)
    Child=(Type=Bone,Name="base_jacket_1_r")
    Parents=((Item=(Type=Bone,Name="pelvis"),Weight=1))
    Weight=@n9.Result
   )
   > &RigUnit_ParentConstraint(
    AdvancedSettings=(InterpolationType=Average,RotationOrderForFilter=XZY)
    Child=(Type=Bone,Name="base_jacket_2_l")
    Parents=((Item=(Type=Bone,Name="pelvis"),Weight=1))
    Weight=@n9.Result
   )
   > &RigUnit_ParentConstraint(
    AdvancedSettings=(InterpolationType=Average,RotationOrderForFilter=XZY)
    Child=(Type=Bone,Name="base_jacket_2_r")
    Parents=((Item=(Type=Bone,Name="pelvis"),Weight=1))
    Weight=@n9.Result
   )
   > RigVMFunctionReturnNode()
  ]}
 New Function_ContainedGraph {nodes [
   RigVMFunctionEntryNode()
   > &RigUnit_ParentConstraint(
    AdvancedSettings=(InterpolationType=Average,RotationOrderForFilter=XZY)
    Child=(Type=Bone,Name="thigh_l_2")
    Parents=((Item=(Type=Bone,Name="thigh_l"),Weight=1))
    Weight=1
   )
   > &RigUnit_ParentConstraint(
    AdvancedSettings=(InterpolationType=Average,RotationOrderForFilter=XZY)
    Child=(Type=Bone,Name="base_strapFrnt_0")
    Parents=((Item=(Type=Bone,Name="thigh_l_2"),Weight=1))
    Weight=1
   )
   > n1 = &RigUnit_SphericalPoseReader(
    ActiveRegionScaleFactors=(PositiveWidth=1,NegativeWidth=1,PositiveHeight=1,NegativeHeight=1)
    ActiveRegionSize=0.1
    Debug=(bDrawDebug=false,bDraw2D=False,bDrawLocalAxes=False,DebugScale=25,DebugSegments=20,DebugThickness=0.25)
    DriverAxis=(X=-1,Y=0,Z=0)
    DriverItem=(Type=Bone,Name="thigh_l")
    FalloffRegionScaleFactors=(PositiveWidth=1,NegativeWidth=2.5,PositiveHeight=2,NegativeHeight=2)
    FalloffSize=0.2
    FlipHeightScaling=False
    FlipWidthScaling=False
    OptionalParentItem=(Type=Bone,Name="None")
    RotationOffset=(X=90,Y=0,Z=0)
   )
   > &RigUnit_ModifyTransforms(
    ItemToModify=((Item=(Type=Bone,Name="base_strapFrnt_0"),Transform=(Rotation=(X=0.3655,Y=0.4173,Z=-0.5436,W=0.6299),Translation=(X=11.6074,Y=-2.7082,Z=8.5299),Scale3D=(X=1,Y=1,Z=1))))
    WeightMinimum=0
    Weight=@n1.OutputParam
   )
   > n2 = &RigUnit_SphericalPoseReader(
    ActiveRegionScaleFactors=(PositiveWidth=1,NegativeWidth=1,PositiveHeight=1,NegativeHeight=1)
    ActiveRegionSize=0.1
    Debug=(bDrawDebug=false,bDraw2D=False,bDrawLocalAxes=False,DebugScale=25,DebugSegments=20,DebugThickness=0.25)
    DriverAxis=(X=-1,Y=0,Z=0)
    DriverItem=(Type=Bone,Name="thigh_l")
    FalloffRegionScaleFactors=(PositiveWidth=1,NegativeWidth=2.8,PositiveHeight=1,NegativeHeight=2.8)
    FalloffSize=0.25
    FlipHeightScaling=False
    FlipWidthScaling=False
    OptionalParentItem=(Type=Bone,Name="None")
    RotationOffset=(X=180,Y=0,Z=0)
   )
   RigVMFunction_AnimEvalRichCurve = &RigVMFunction_AnimEvalRichCurve(
    Curve="(EditorCurveData=(Keys=((InterpMode=RCIM_Cubic,TangentMode=RCTM_User,TangentWeightMode=RCTWM_WeightedNone,Time=0,Value=0,ArriveTangent=0.4459,ArriveTangentWeight=0,LeaveTangent=0.4459,LeaveTangentWeight=0),(InterpMode=RCIM_Cubic,TangentMode=RCTM_User,Tange ...(truncated)"
    SourceMinimum=0
    TargetMinimum=0
    Value=@n2.OutputParam
   )
   > &RigUnit_ModifyTransforms(
    ItemToModify="((Item=(Type=Bone,Name=\"base_strapFrnt_0\"),Transform=(Rotation=(X=-0.0721,Y=-0.1877,Z=-0.5415,W=0.8163),Translation=(X=0,Y=0,Z=0),Scale3D=(X=1,Y=1,Z=1))),(Item=(Type=Bone,Name=\"base_strapBck_0\"),Transform=(Rotation=(X=-0.0062,Y=-0.0031,Z=-0.6593,W=0.7519), ...(truncated)"
    WeightMinimum=0
    Weight=@RigVMFunction_AnimEvalRichCurve.Result
   )
   RigVMFunction_MathBoolToFloat = &RigVMFunction_MathBoolToFloat(Value=RigVMVariableNode() {variable=isCrouching}.Value)
   > &RigUnit_ModifyTransforms(
    ItemToModify="((Item=(Type=Bone,Name=\"base_strapFrnt_0\"),Transform=(Rotation=(X=0.0748,Y=-0.3393,Z=-0.1757,W=0.9211),Translation=(X=-5.6947,Y=8.8708,Z=-11.2028),Scale3D=(X=1,Y=1,Z=1))),(Item=(Type=Bone,Name=\"dyn_strapFrnt_00_fix\"),Transform=(Rotation=(X=0.0893,Y=0.2458, ...(truncated)"
    WeightMinimum=0
    Weight=@RigVMFunction_MathBoolToFloat.Result
   )
   > &RigUnit_ModifyTransforms(
    ItemToModify=((Item=(Type=Bone,Name="base_strapFrnt_0"),Transform=(Rotation=(X=0.0504,Y=0.7713,Z=-0.2965,W=0.561),Translation=(X=20.6536,Y=-2.6252,Z=7.9415),Scale3D=(X=1,Y=1,Z=1))))
    WeightMinimum=0
    Weight=&RigVMFunction_MathBoolToFloat(Value=RigVMVariableNode() {variable=isSliding}.Value).Result
   )
   > n3 = &RigUnit_SphericalPoseReader(
    ActiveRegionScaleFactors=(PositiveWidth=1,NegativeWidth=1,PositiveHeight=1,NegativeHeight=1)
    ActiveRegionSize=0.1
    Debug=(bDrawDebug=false,bDraw2D=False,bDrawLocalAxes=False,DebugScale=25,DebugSegments=20,DebugThickness=0.25)
    DriverAxis=(X=-1,Y=0,Z=0)
    DriverItem=(Type=Bone,Name="thigh_l")
    FalloffRegionScaleFactors=(PositiveWidth=1,NegativeWidth=1,PositiveHeight=1,NegativeHeight=1)
    FalloffSize=0.2
    FlipHeightScaling=False
    FlipWidthScaling=False
    OptionalParentItem=(Type=Bone,Name="None")
    RotationOffset=(X=80,Y=0,Z=70)
   )
   Multiply = &RigVMFunction_MathDoubleMul(B=&RigVMFunction_MathBoolToFloat(Value=RigVMVariableNode() {variable=isCrouchingSprinting}.Value).Result, A=@n3.OutputParam)
   > &RigUnit_ModifyTransforms(
    ItemToModify=((Item=(Type=Bone,Name="base_strapFrnt_0"),Transform=(Rotation=(X=-0.2027,Y=0.0839,Z=-0.0711,W=0.973),Translation=(X=8.2453,Y=-2.2634,Z=1.3994),Scale3D=(X=1,Y=1,Z=1))))
    WeightMinimum=0
    Weight=@Multiply.Result
   )
   > RigVMFunctionReturnNode()
  ]}
 RigVMFunctionLibrary {nodes [
   controls = RigVMCollapseNode(ExecuteContext=(), debug=False)
   foot band fixes = RigVMCollapseNode(ExecuteContext=())
   jacket fixes = RigVMCollapseNode(ExecuteContext=())
   strap fixes = RigVMCollapseNode(ExecuteContext=())
  ]}
 RigVMFunction_Sequence_1_ContainedGraph {nodes [
   n1 = RigVMFunctionEntryNode()
   n2 = &RigVMFunction_Sequence(ExecuteContext=@n1.ExecuteContext)
   n3 = &RigVMFunction_Sequence(ExecuteContext=@n2.B)
   n4 = &RigVMFunction_Sequence(ExecuteContext=@n3.B)
   n5 = &RigVMFunction_Sequence(ExecuteContext=@n4.B)
   RigVMFunctionReturnNode(B=@n2.A, G=@n5.A, H=@n5.B, E=@n4.A, C=@n3.A)
  ]}
 RigVMModel {nodes [
   RigUnit_BeginExecution = &RigUnit_BeginExecution()
   RigVMFunction_Sequence = &RigVMFunction_Sequence(ExecuteContext=@RigUnit_BeginExecution.ExecutePin)
   controls = RigVMFunctionReferenceNode(debug=false, ExecuteContext=@RigVMFunction_Sequence.A)
   FN_DeformRig_FNC = RigVMFunctionReferenceNode(
    DeformRig_ Curve="(deformRig_l_upperarm_up_alpha_72_#1=False,deformRig_r_upperarm_up_alpha_73_#2=False,deformRig_l_upperarm_fr_alpha_74_#3=False,deformRig_r_upperarm_fr_alpha_75_CB51B8 ...(truncated)"
    DeformRig_Bool="(upperarm_1_#4=True,elbow_3_#5=True,wrist_7_#6=True,groin_8_#7=True,glute_17_#8=True,knee_9_40C408324CA261 ...(truncated)"
    DeformRig_Debug=(upperarm_debug_18_#4=False,wrist_debug_20_#6=False,lat_debug_26_#9=False,groinGlute_debug_27_#7=false)
    DeformRig_Value="(upperarm_up_18_#10=(Rotation=(X=0.0,Y=-0.0,Z=0.0,W=1.0),Translation=(X=0.0,Y=0.0,Z=8.0),Scale3D=(X=0.70,Y=1.0,Z=1.0)),upperarm_fwd_fr_19_#11=(Rotation=(X=0.0,Y=-0.0,Z=0.0,W=1.0),Translation=(X=0.0, ...(truncated)"
    ExecuteContext=@RigVMFunction_Sequence.B
   )
   > foot band fixes = RigVMFunctionReferenceNode()
   > jacket fixes = RigVMFunctionReferenceNode()
   > strap fixes = RigVMFunctionReferenceNode()
  ]}
]
hierarchy [
 root {loc=<0 0 0>, rot=<0 0 0 1>}
 pelvis {parent=root, loc=<0 -0.5301 98.0571>, rot=<0.0127 0.707 -0.0127 -0.707>}
 spine_01 {parent=pelvis, loc=<6.1208 -0.9328 0>, rot=<0 0 0.034 -0.9994>}
 spine_02 {parent=spine_01, loc=<6.5997 -0.2095 0>, rot=<0 0 0.0238 -0.9997>}
 spine_03 {parent=spine_02, loc=<6.4914 -0.1237 0>, rot=<0 0 0.0149 -0.9999>}
 spine_04 {parent=spine_03, loc=<6.9598 0.0794 0>, rot=<0 0 -0.0132 -0.9999>}
 spine_05 {parent=spine_04, loc=<6.9987 0 0>, rot=<0 0 -0.0015 -1>}
 attach_cape {parent=spine_05, loc=<9.0165 -11.7139 0.0001>, rot=<0.0616 -0.7044 -0.0616 -0.7044>}
 neck_01 {parent=spine_05, loc=<11.3138 0.1529 0>, rot=<0 0 -0.2003 -0.9797>}
 neck_02 {parent=neck_01, loc=<4.5627 0 0>, rot=<0 0 0.0504 -0.9987>}
 head {parent=neck_02, loc=<5.3914 0 0>, rot=<0 0 0.115 -0.9934>}
 FACIAL_C_FacialRoot {parent=head, loc=<2.0085 0.1746 -0.0002>, rot=<0.5021 -0.4979 -0.5021 -0.4979>}
 FACIAL_C_UnderChin {parent=FACIAL_C_FacialRoot, loc=<-0.0002 8.7048 3.5163>, rot=<0.5 0 0 -0.866>}
 FACIAL_L_UnderChin {parent=FACIAL_C_FacialRoot, loc=<2.7308 7.4531 2.1505>, rot=<0.4025 -0.355 0.0823 -0.8398>}
 FACIAL_R_UnderChin {parent=FACIAL_C_FacialRoot, loc=<-2.7308 7.4531 2.1505>, rot=<0.4025 0.355 -0.0823 -0.8398>}
 FACIAL_C_Jawline {parent=FACIAL_C_FacialRoot, loc=<-0.0002 8.4758 5.393>, rot=<0.5736 0 0 -0.8192>}
 FACIAL_L_Jawline {parent=FACIAL_C_FacialRoot, loc=<3.9122 5.4242 4.0145>, rot=<0.2092 0.1578 -0.0547 -0.9635>}
 FACIAL_R_Jawline {parent=FACIAL_C_FacialRoot, loc=<-3.9122 5.4242 4.0145>, rot=<0.2092 -0.1578 0.0547 -0.9635>}
 attach_backpack {parent=spine_05, loc=<-0.1231 -13.4729 0.0002>, rot=<-0.0093 -0.707 0.0093 -0.707>}
 clavicle_l {parent=spine_05, loc=<5.4831 -0.1053 -3.2946>, rot=<0.0465 -0.7289 -0.009 -0.6829>}
 upperarm_l {parent=clavicle_l, loc=<9.8658 -1.2136 -0.4105>, rot=<0.0164 -0.3245 -0.0269 -0.9454>}
 lowerarm_l {parent=upperarm_l, loc=<26.465 -0.8345 0.259>, rot=<0.0036 0.0161 -0.2298 -0.9731>}
 hand_l {parent=lowerarm_l, loc=<22.7355 -0.3203 -0.6105>, rot=<-0.6784 -0.0223 0.0735 -0.7307>}
 thumb_01_l {parent=hand_l, loc=<1.3118 -0.3191 -2.459>, rot=<0.6546 -0.4042 0.0664 -0.6354>}
 thumb_02_l {parent=thumb_01_l, loc=<5.9336 -0.0448 0.0401>, rot=<-0.0037 0.0213 0.298 -0.9543>}
 thumb_03_l {parent=thumb_02_l, loc=<2.2223 0.3682 -0.0907>, rot=<0.0118 -0.013 0.0066 -0.9998>}
 index_metacarpal_l {parent=hand_l, loc=<1.6535 0.0183 -2.96>, rot=<0.0158 -0.0124 -0.0173 -0.9996>}
 index_01_l {parent=index_metacarpal_l, loc=<8.3008 0.0357 -0.0524>, rot=<-0.0044 0.0485 0.0924 -0.9945>}
 index_02_l {parent=index_01_l, loc=<3.3587 -0.3004 0.0452>, rot=<-0.002 0.0027 0.3225 -0.9466>}
 index_03_l {parent=index_02_l, loc=<2.5578 -0.123 -0.0252>, rot=<-0.002 0.0067 0.2558 -0.9667>}
 middle_metacarpal_l {parent=hand_l, loc=<2.1856 -0.1725 -0.5984>, rot=<0.006 0.0168 -0.0329 -0.9993>}
 middle_01_l {parent=middle_metacarpal_l, loc=<7.3275 -0.2447 0.0129>, rot=<-0.0208 0.0042 0.0858 -0.9961>}
 middle_02_l {parent=middle_01_l, loc=<4.0123 0.1401 0.1175>, rot=<-0.0218 0.0258 0.3512 -0.9357>}
 middle_03_l {parent=middle_02_l, loc=<2.6077 0.0534 0.1373>, rot=<0.0082 0.0188 0.1708 -0.9851>}
 ring_metacarpal_l {parent=hand_l, loc=<2.2074 -0.3625 1.3638>, rot=<-0.0284 0.0722 -0.0458 -0.9959>}
 ring_01_l {parent=ring_metacarpal_l, loc=<7.0295 -0.6197 -0.1446>, rot=<-0.0489 -0.0252 0.1892 -0.9804>}
 ring_02_l {parent=ring_01_l, loc=<3.6174 0.4647 -0.1027>, rot=<0 0 0.2806 -0.9598>}
 ring_03_l {parent=ring_02_l, loc=<2.4152 0.114 0.0199>, rot=<0 0 0.1189 -0.9929>}
 pinky_metacarpal_l {parent=hand_l, loc=<1.7616 -0.9753 3.3814>, rot=<-0.0397 0.0814 -0.069 -0.9935>}
 pinky_01_l {parent=pinky_metacarpal_l, loc=<7.1709 -0.4831 -0.1094>, rot=<-0.0654 -0.0035 0.2242 -0.9723>}
 pinky_02_l {parent=pinky_01_l, loc=<2.8274 0.0893 -0.1043>, rot=<0.0037 0.0011 0.2854 -0.9584>}
 pinky_03_l {parent=pinky_02_l, loc=<2.2046 0.0307 -0.0078>, rot=<-0.0002 0.0098 0.1445 -0.9895>}
 deform_wrist_dn_base_l {parent=hand_l, loc=<0 -1.5326 0>, rot=<0 0 0 -1>}
 deform_wrist_dn_l {parent=deform_wrist_dn_base_l, loc=<0 0 0>, rot=<0 0 0 1>}
 deform_wrist_up_base_l {parent=hand_l, loc=<0 2.3009 0>, rot=<0 0 0 -1>}
 deform_wrist_up_l {parent=deform_wrist_up_base_l, loc=<0 0 0>, rot=<0 0 0 1>}
 lowerarm_twist_02_l {parent=lowerarm_l, loc=<10.6707 0 0>, rot=<0 0 0 1>}
 lowerarm_twist_01_l {parent=lowerarm_l, loc=<18.316 0 0>, rot=<0 0 0 1>}
 deform_elbow_in_base_l {parent=lowerarm_l, loc=<0.1834 3.6062 -0.2572>, rot=<0 0 0 -1>}
 deform_elbow_in_l {parent=deform_elbow_in_base_l, loc=<0 0 0>, rot=<0 0 0 1>}
 deform_elbow_out_base_l {parent=lowerarm_l, loc=<0 -4 0>, rot=<0 0 0 1>}
 deform_elbow_out_l {parent=deform_elbow_out_base_l, loc=<0 0 0>, rot=<0 0 0 1>}
 upperarm_twist_01_l {parent=upperarm_l, loc=<7.2135 -0.1134 0>, rot=<0 0 0 1>}
 deform_upperarm_bk_base_l {parent=upperarm_twist_01_l, loc=<-6.4053 -4.3746 0.314>, rot=<0 0 0 -1>}
 deform_upperarm_bk_l {parent=deform_upperarm_bk_base_l, loc=<0 0 0>, rot=<0 0 0 1>}
 deform_upperarm_fr_base_l {parent=upperarm_twist_01_l, loc=<-5.6312 5.092 0.3354>, rot=<0 0 0 -1>}
 deform_upperarm_fr_l {parent=deform_upperarm_fr_base_l, loc=<0 0 0>, rot=<0 0 0 1>}
 deform_upperarm_up_base_l {parent=upperarm_twist_01_l, loc=<-8.891 0.4825 3.9517>, rot=<0 0 0 -1>}
 deform_upperarm_up_l {parent=deform_upperarm_up_base_l, loc=<0 0 0>, rot=<0 0 0 1>}
 upperarm_twist_02_l {parent=upperarm_l, loc=<19.7998 -0.1134 0>, rot=<0 0 0 1>}
 clavicle_r {parent=spine_05, loc=<5.4831 -0.1053 3.2946>, rot=<0.7289 0.0465 -0.6829 0.009>}
 upperarm_r {parent=clavicle_r, loc=<-9.8658 1.2136 0.4105>, rot=<0.0164 -0.3245 -0.0269 -0.9454>}
 lowerarm_r {parent=upperarm_r, loc=<-26.465 0.8345 -0.259>, rot=<0.0036 0.0161 -0.2298 -0.9731>}
 hand_r {parent=lowerarm_r, loc=<-22.7355 0.3203 0.6106>, rot=<-0.6784 -0.0223 0.0735 -0.7307>}
 thumb_01_r {parent=hand_r, loc=<-1.3118 0.3191 2.459>, rot=<0.6546 -0.4042 0.0664 -0.6354>}
 thumb_02_r {parent=thumb_01_r, loc=<-5.9336 0.0448 -0.0401>, rot=<-0.0037 0.0213 0.298 -0.9543>}
 thumb_03_r {parent=thumb_02_r, loc=<-2.2223 -0.3682 0.0907>, rot=<0.0118 -0.013 0.0066 -0.9998>}
 index_metacarpal_r {parent=hand_r, loc=<-1.6535 -0.0183 2.96>, rot=<0.0158 -0.0124 -0.0173 -0.9996>}
 index_01_r {parent=index_metacarpal_r, loc=<-8.3008 -0.0357 0.0524>, rot=<-0.0044 0.0485 0.0924 -0.9945>}
 index_02_r {parent=index_01_r, loc=<-3.3587 0.3004 -0.0452>, rot=<-0.002 0.0027 0.3225 -0.9466>}
 index_03_r {parent=index_02_r, loc=<-2.5578 0.123 0.0252>, rot=<-0.002 0.0067 0.2558 -0.9667>}
 middle_metacarpal_r {parent=hand_r, loc=<-2.1856 0.1725 0.5984>, rot=<0.006 0.0168 -0.0329 -0.9993>}
 middle_01_r {parent=middle_metacarpal_r, loc=<-7.3275 0.2447 -0.0129>, rot=<-0.0208 0.0042 0.0858 -0.9961>}
 middle_02_r {parent=middle_01_r, loc=<-4.0123 -0.1401 -0.1175>, rot=<-0.0218 0.0258 0.3512 -0.9357>}
 middle_03_r {parent=middle_02_r, loc=<-2.6077 -0.0534 -0.1373>, rot=<0.0082 0.0188 0.1708 -0.9851>}
 ring_metacarpal_r {parent=hand_r, loc=<-2.2074 0.3625 -1.3638>, rot=<-0.0284 0.0722 -0.0458 -0.9959>}
 ring_01_r {parent=ring_metacarpal_r, loc=<-7.0295 0.6197 0.1446>, rot=<-0.0489 -0.0252 0.1892 -0.9804>}
 ring_02_r {parent=ring_01_r, loc=<-3.6174 -0.4647 0.1027>, rot=<0 0 0.2806 -0.9598>}
 ring_03_r {parent=ring_02_r, loc=<-2.4152 -0.114 -0.0199>, rot=<0 0 0.1189 -0.9929>}
 pinky_metacarpal_r {parent=hand_r, loc=<-1.7616 0.9753 -3.3814>, rot=<-0.0397 0.0814 -0.069 -0.9935>}
 pinky_01_r {parent=pinky_metacarpal_r, loc=<-7.1709 0.4831 0.1094>, rot=<-0.0654 -0.0035 0.2242 -0.9723>}
 pinky_02_r {parent=pinky_01_r, loc=<-2.8274 -0.0893 0.1043>, rot=<0.0037 0.0011 0.2854 -0.9584>}
 pinky_03_r {parent=pinky_02_r, loc=<-2.2046 -0.0307 0.0078>, rot=<-0.0002 0.0098 0.1445 -0.9895>}
 deform_wrist_dn_base_r {parent=hand_r, loc=<-0.0001 1.5328 0>, rot=<0 0 0 -1>}
 deform_wrist_dn_r {parent=deform_wrist_dn_base_r, loc=<0 0 0>, rot=<0 0 0 1>}
 deform_wrist_up_base_r {parent=hand_r, loc=<-0.0002 -2.3007 -0.0001>, rot=<0 0 0 -1>}
 deform_wrist_up_r {parent=deform_wrist_up_base_r, loc=<0 0 0>, rot=<0 0 0 1>}
 lowerarm_twist_02_r {parent=lowerarm_r, loc=<-10.6707 0 0>, rot=<0 0 0 -1>}
 lowerarm_twist_01_r {parent=lowerarm_r, loc=<-18.316 0 0>, rot=<0 0 0 -1>}
 deform_elbow_in_base_r {parent=lowerarm_r, loc=<-0.1834 -3.6062 0.2572>, rot=<0 0 0 -1>}
 deform_elbow_in_r {parent=deform_elbow_in_base_r, loc=<0 0 0>, rot=<0 0 0 1>}
 deform_elbow_out_base_r {parent=lowerarm_r, loc=<0 4 -0.0001>, rot=<0 0 0 -1>}
 deform_elbow_out_r {parent=deform_elbow_out_base_r, loc=<0 0 0>, rot=<0 0 0 1>}
 upperarm_twist_01_r {parent=upperarm_r, loc=<-7.2135 0.1134 0>, rot=<0 0 0 -1>}
 deform_upperarm_bk_base_r {parent=upperarm_twist_01_r, loc=<6.4053 4.3746 -0.314>, rot=<0 0 0 -1>}
 deform_upperarm_bk_r {parent=deform_upperarm_bk_base_r, loc=<0 0 0>, rot=<0 0 0 1>}
 deform_upperarm_fr_base_r {parent=upperarm_twist_01_r, loc=<5.6312 -5.092 -0.3354>, rot=<0 0 0 -1>}
 deform_upperarm_fr_r {parent=deform_upperarm_fr_base_r, loc=<0 0 0>, rot=<0 0 0 1>}
 deform_upperarm_up_base_r {parent=upperarm_twist_01_r, loc=<8.891 -0.4825 -3.9517>, rot=<0 0 0 -1>}
 deform_upperarm_up_r {parent=deform_upperarm_up_base_r, loc=<0 0 0>, rot=<0 0 0 1>}
 upperarm_twist_02_r {parent=upperarm_r, loc=<-19.7998 0.1134 0>, rot=<0 0 0 -1>}
 Z_aim_hair {parent=spine_05, loc=<1.8481 -11.1302 -2.3091>, rot=<0.0284 -0.7065 -0.0284 -0.7065>}
 deform_pec_aim_base_l {parent=spine_04, loc=<3.2456 9.3309 -1.2237>, rot=<0.0603 -0.4033 0.3318 -0.8507>}
 deform_pec_aim_l {parent=deform_pec_aim_base_l, loc=<0 0 0>, rot=<0 0 0 1>}
 deform_pec_base_l {parent=deform_pec_aim_l, loc=<9.3665 0 0>, rot=<0 0 0 1>}
 deform_pec_l {parent=deform_pec_base_l, loc=<0 0 0>, rot=<0 0 0 1>}
 deform_pec_aim_base_r {parent=spine_04, loc=<3.2452 9.3309 1.2237>, rot=<-0.4033 -0.0603 0.8507 0.3318>}
 deform_pec_aim_r {parent=deform_pec_aim_base_r, loc=<0 0 0>, rot=<0 0 0 1>}
 deform_pec_base_r {parent=deform_pec_aim_r, loc=<-9.3667 -0.0003 -0.0004>, rot=<0 0 0 1>}
 deform_pec_r {parent=deform_pec_base_r, loc=<0 0 0>, rot=<0 0 0 1>}
 base_handle_chest_strap_r {parent=spine_04, loc=<9.8065 10.9834 5.8794>, rot=<0.2569 0.4923 0.8315 0.0148>}
 dyn_handle_chest_strap_r {parent=base_handle_chest_strap_r, loc=<0 0 0>, rot=<0 0 0 1>}
 base_chest_strap0_r {parent=dyn_handle_chest_strap_r, loc=<4.9239 0.0066 -0.012>, rot=<-0.065 -0.0964 0.0403 -0.9924>}
 dyn_chest_strap0_r {parent=base_chest_strap0_r, loc=<0 0 0>, rot=<0 0 0 1>}
 dyn_chest_strap1_r {parent=dyn_chest_strap0_r, loc=<8 0 0>, rot=<-0.0051 -0.1595 0.0316 -0.9867>}
 dyn_chest_strap2_r {parent=dyn_chest_strap1_r, loc=<6.9676 0 0>, rot=<0 -0.0349 0 -0.9994>}
 base_handle_chest_strap_l {parent=spine_04, loc=<9.8062 10.9834 -5.8794>, rot=<-0.0073 -0.8149 -0.5188 -0.2584>}
 dyn_handle_chest_strap_l {parent=base_handle_chest_strap_l, loc=<0 0 0>, rot=<0 0 0 1>}
 base_chest_strap0_l {parent=dyn_handle_chest_strap_l, loc=<4.9618 -0.0018 -0.0654>, rot=<0.0036 -0.1117 -0.0352 -0.9931>}
 dyn_chest_strap0_l {parent=base_chest_strap0_l, loc=<0 0 0>, rot=<0 0 0 1>}
 dyn_chest_strap1_l {parent=dyn_chest_strap0_l, loc=<9.3902 0 0>, rot=<0.006 -0.1741 -0.0397 -0.9839>}
 dyn_chest_strap2_l {parent=dyn_chest_strap1_l, loc=<11.0629 0 0>, rot=<0.0003 -0.0153 -0.0164 -0.9997>}
 base_jacket_0_l {parent=spine_03, loc=<-6.3125 11.6073 -7.0261>, rot=<0.0097 0.6543 0.7562 -0.0021>}
 dyn_jacket_0_l {parent=base_jacket_0_l, loc=<0 0 0>, rot=<0 0 0 1>}
 dyn_jacket_0_end_l {parent=dyn_jacket_0_l, loc=<12.3371 0 0>, rot=<0 0 0 1>}
 base_jacket_0_r {parent=spine_03, loc=<-6.3124 11.6073 7.0261>, rot=<-0.0097 -0.6543 0.7562 -0.0021>}
 dyn_jacket_0_r {parent=base_jacket_0_r, loc=<0 0 0>, rot=<0 0 0 1>}
 dyn_jacket_0_end_r {parent=dyn_jacket_0_r, loc=<12.337 0 0>, rot=<0 0 0 1>}
 base_jacket_1_r {parent=spine_03, loc=<-4.6831 4.3482 14.4099>, rot=<0.1781 -0.6205 0.7628 0.0369>}
 dyn_jacket_1_r {parent=base_jacket_1_r, loc=<0 0 0>, rot=<0 0 0 1>}
 dyn_jacket_1_end_r {parent=dyn_jacket_1_r, loc=<11.915 -0.0003 -0.0001>, rot=<0 0 0 1>}
 base_jacket_2_r {parent=spine_03, loc=<-2.5296 -6.8922 12.8378>, rot=<0.1819 -0.8273 0.5308 0.028>}
 dyn_jacket_2_r {parent=base_jacket_2_r, loc=<0 0 0>, rot=<0 0 0 -1>}
 dyn_jacket_2_end_r {parent=dyn_jacket_2_r, loc=<11.543 0 0>, rot=<0 0 0 1>}
 base_jacket_0_m {parent=spine_03, loc=<-1.3254 -11.9438 -0.1388>, rot=<0.2358 -0.9718 0 0>}
 dyn_jacket_0_m {parent=base_jacket_0_m, loc=<0 0 0>, rot=<0 0 0 1>}
 dyn_jacket_0_end_m {parent=dyn_jacket_0_m, loc=<11.5538 0 0>, rot=<0 0 0 1>}
 base_jacket_2_l {parent=spine_03, loc=<-2.5299 -6.8922 -12.8378>, rot=<0.1819 -0.8273 -0.5308 -0.028>}
 dyn_jacket_2_l {parent=base_jacket_2_l, loc=<0 0 0>, rot=<0 0 0 1>}
 dyn_jacket_2_end_l {parent=dyn_jacket_2_l, loc=<11.5435 0 0>, rot=<0 0 0 1>}
 base_jacket_1_l {parent=spine_03, loc=<-4.6836 4.3482 -14.4099>, rot=<-0.1781 0.6205 0.7628 0.0369>}
 dyn_jacket_1_l {parent=base_jacket_1_l, loc=<0 0 0>, rot=<0 0 0 1>}
 dyn_jacket_1_end_l {parent=dyn_jacket_1_l, loc=<11.9137 0 0>, rot=<0 0 0 1>}
 thigh_l {parent=pelvis, loc=<-1.0855 -0.2109 -8.7053>, rot=<0.0009 0.0355 0.0155 -0.9993>}
 calf_l {parent=thigh_l, loc=<-42.5284 -0.0019 -0.7297>, rot=<0.0006 -0.0161 -0.0391 -0.9991>}
 foot_l {parent=calf_l, loc=<-46.7401 -2.5014 -0.592>, rot=<0.0427 -0.0238 0.0353 -0.9982>}
 ball_l {parent=foot_l, loc=<-3.5658 10.8066 -0.1544>, rot=<0.0001 -0.0001 -0.7063 -0.7079>}
 base_shoeband_l {parent=foot_l, loc=<-4.0293 1.5554 0.0616>, rot=<-0.6619 -0.2247 -0.2518 -0.6693>}
 dyn_shoeband_l {parent=base_shoeband_l, loc=<0 0 0>, rot=<0 0 0 1>}
 calf_twist_02_l {parent=calf_l, loc=<-19.1546 -1.7943 -0.2638>, rot=<0.0237 0.0095 -0.0035 -0.9997>}
 calf_twist_01_l {parent=calf_l, loc=<-35.8483 -1.9197 -0.5794>, rot=<0.0237 0.0095 -0.0035 -0.9997>}
 deform_knee_in_base_l {parent=calf_l, loc=<0.8031 -5.0231 0.2456>, rot=<0 0 0 -1>}
 deform_knee_in_l {parent=deform_knee_in_base_l, loc=<0 0 0>, rot=<0 0 0 1>}
 deform_knee_out_base_l {parent=calf_l, loc=<0 4 0>, rot=<0 0 0 -1>}
 deform_knee_out_l {parent=deform_knee_out_base_l, loc=<0 0 0>, rot=<0 0 0 1>}
 foot_l_get_rot_offset {parent=calf_l, loc=<-46.7401 -2.5014 -0.592>, rot=<0.0427 -0.0238 0.0353 -0.9982>}
 foot_l_get_rot {parent=foot_l_get_rot_offset, loc=<0 0 0>, rot=<0 0 0 1>}
 thigh_twist_01_l {parent=thigh_l, loc=<-27.952 0 0>, rot=<0.0013 0 0 -1>}
 deform_groin_base_l {parent=thigh_l, loc=<0.4801 5.6776 -0.2486>, rot=<0 0 0 -1>}
 deform_groin_l {parent=deform_groin_base_l, loc=<0 0 0>, rot=<0 0 0 1>}
 base_strapFrnt_0 {parent=thigh_l, loc=<3.9129 7.8422 -0.1599>, rot=<0.0957 -0.8207 -0.5553 -0.0944>}
 dyn_strapFrnt_00 {parent=base_strapFrnt_0, loc=<0 0 0>, rot=<0 0 0 1>}
 dyn_strapFrnt_00_fix {parent=dyn_strapFrnt_00, loc=<4.1828 0 0>, rot=<0 0 0 1>}
 dyn_strapFrnt_01 {parent=dyn_strapFrnt_00_fix, loc=<4.1828 0 0>, rot=<0 -0.016 -0.0074 -0.9998>}
 dyn_strapFrnt_02 {parent=dyn_strapFrnt_01, loc=<9.1411 0 0>, rot=<0.0004 -0.0423 0.0244 -0.9988>}
 dyn_strapFrnt_03 {parent=dyn_strapFrnt_02, loc=<9.6448 0 0>, rot=<-0.008 -0.1732 0.0614 -0.9829>}
 dyn_strapFrnt_04 {parent=dyn_strapFrnt_03, loc=<6.1805 0 0>, rot=<0 0 0 1>}
 thigh_r {parent=pelvis, loc=<-1.0855 -0.2109 8.7053>, rot=<0.0355 -0.0009 0.9993 0.0155>}
 calf_r {parent=thigh_r, loc=<42.5284 0.0019 0.7297>, rot=<0.0006 -0.0161 -0.0391 -0.9991>}
 foot_r {parent=calf_r, loc=<46.7401 2.5014 0.592>, rot=<0.0427 -0.0238 0.0353 -0.9982>}
 ball_r {parent=foot_r, loc=<3.5658 -10.8066 0.1544>, rot=<0.0001 -0.0001 -0.7063 -0.7079>}
 base_shoeband_r {parent=foot_r, loc=<4.0293 -1.5554 -0.0616>, rot=<-0.6619 -0.2247 -0.2518 -0.6693>}
 dyn_shoeband_r {parent=base_shoeband_r, loc=<0 0 0>, rot=<0 0 0 1>}
 calf_twist_02_r {parent=calf_r, loc=<19.1546 1.7943 0.2638>, rot=<0.0237 0.0095 -0.0035 -0.9997>}
 calf_twist_01_r {parent=calf_r, loc=<35.8483 1.9197 0.5794>, rot=<0.0237 0.0095 -0.0035 -0.9997>}
 deform_knee_in_base_r {parent=calf_r, loc=<-0.8031 5.0231 -0.2457>, rot=<0 0 0 -1>}
 deform_knee_in_r {parent=deform_knee_in_base_r, loc=<0 0 0>, rot=<0 0 0 1>}
 deform_knee_out_base_r {parent=calf_r, loc=<0 -4 0>, rot=<0 0 0 -1>}
 deform_knee_out_r {parent=deform_knee_out_base_r, loc=<0 0 0>, rot=<0 0 0 1>}
 foot_r_get_rot_offset {parent=calf_r, loc=<46.7401 2.5014 0.592>, rot=<0.0427 -0.0238 0.0353 -0.9982>}
 foot_r_get_rot {parent=foot_r_get_rot_offset, loc=<0 0 0>, rot=<0 0 0 1>}
 thigh_twist_01_r {parent=thigh_r, loc=<27.952 0 0>, rot=<0.0013 0 0 -1>}
 deform_groin_base_r {parent=thigh_r, loc=<-0.4801 -5.6776 0.2486>, rot=<0 0 0 -1>}
 deform_groin_r {parent=deform_groin_base_r, loc=<0 0 0>, rot=<0 0 0 1>}
 base_bag_r {parent=thigh_r, loc=<22.1784 -0.1274 7.7732>, rot=<-0.0004 -0.079 0.0024 -0.9969>}
 dyn_bag_r {parent=base_bag_r, loc=<0 0 0>, rot=<0 0 0 1>}
 deform_glute_base_l {parent=pelvis, loc=<-7.3466 -9.0213 -9.1245>, rot=<0.0172 0.031 -0.4477 -0.8935>}
 deform_glute_l {parent=deform_glute_base_l, loc=<0 0 0>, rot=<0 0 0 1>}
 deform_glute_base_r {parent=pelvis, loc=<-7.3464 -9.0213 9.1245>, rot=<0.031 -0.0172 0.8935 -0.4477>}
 deform_glute_r {parent=deform_glute_base_r, loc=<0 0 0>, rot=<0 0 0 1>}
 base_main_zipper {parent=pelvis, loc=<-11.9101 12.3543 -9.6635>, rot=<-0.3452 -0.5818 -0.3816 -0.6298>}
 base_zipper {parent=base_main_zipper, loc=<0.358 -0.0021 -0.0251>, rot=<-0.0478 -0.6492 -0.0179 -0.7589>}
 dyn_zipper {parent=base_zipper, loc=<0 0 0>, rot=<0 0 0 1>}
 base_hip_strap1_l {parent=pelvis, loc=<9.9787 -15.9775 -4.6416>, rot=<0.0949 -0.7314 0.6673 -0.1037>}
 dyn_hip_strap1_l {parent=base_hip_strap1_l, loc=<0 0 0>, rot=<0 0 0 1>}
 dyn_hip_strap2_l {parent=dyn_hip_strap1_l, loc=<10.34 0 0>, rot=<-0.0027 -0.125 0.002 -0.9921>}
 dyn_hip_strap3_l {parent=dyn_hip_strap2_l, loc=<6.601 0 0>, rot=<0.0001 -0.0275 -0.0025 -0.9996>}
 dyn_hip_strap4_l {parent=dyn_hip_strap3_l, loc=<6.3812 0 0>, rot=<-0.0005 -0.0079 0.0132 -0.9999>}
 base_hip_strap1_r {parent=pelvis, loc=<9.9715 -15.9779 4.6124>, rot=<0.1009 -0.6555 0.7425 -0.094>}
 dyn_hip_strap1_r {parent=base_hip_strap1_r, loc=<0 0 0>, rot=<0 0 0 1>}
 dyn_hip_strap2_r {parent=dyn_hip_strap1_r, loc=<10.2518 0 0.0624>, rot=<-0.0002 -0.12 0.0073 -0.9927>}
 dyn_hip_strap3_r {parent=dyn_hip_strap2_r, loc=<8.8174 0 0>, rot=<0.0001 -0.0407 -0.0017 -0.9992>}
 dyn_hip_strap4_r {parent=dyn_hip_strap3_r, loc=<8.3676 0 0>, rot=<0 0.0033 0 -1>}
 base_strapBck_0 {parent=pelvis, loc=<4.2557 -10.0424 -5.3859>, rot=<0.2015 -0.7585 -0.6195 -0.0194>}
 dyn_strapBck_00 {parent=base_strapBck_0, loc=<0 0 0>, rot=<0 0 0 1>}
 dyn_strapBck_01 {parent=dyn_strapBck_00, loc=<8.6473 0 0>, rot=<0 0.1912 0 -0.9816>}
 dyn_strapBck_02 {parent=dyn_strapBck_01, loc=<11.149 0 0>, rot=<-0.0442 0.0788 0.0035 -0.9959>}
 dyn_strapBck_03 {parent=dyn_strapBck_02, loc=<7.6942 0 0>, rot=<0.0013 0.0884 -0.0001 -0.9961>}
 dyn_strapBck_04 {parent=dyn_strapBck_03, loc=<8.412 0 0>, rot=<0 0 0 1>}
 Aim_hip_straps {parent=pelvis, loc=<-10.5276 -17.7707 -0.0833>, rot=<0.0011 -0.7387 0.674 -0.0009>}
 thigh_l_2 {parent=pelvis, loc=<-1.0855 -0.2109 -8.7053>, rot=<0.0009 0.0355 0.0155 -0.9992>}
 foot_l_ctrl {controlType=EULER_TRANSFORM, type=Control, loc=<0 0 0>, rot=<0 0 0.0072 1>}
 foot_r_ctrl {controlType=EULER_TRANSFORM, type=Control, loc=<0 0 0>, rot=<-0.0059 -0.7285 -0.0681 0.6816>}
 thigh_l_ctrl {controlType=EULER_TRANSFORM, type=Control, loc=<0 0 0>, rot=<0 0 -0.6667 0.7453>}
 thigh_r_ctrl {controlType=EULER_TRANSFORM, type=Control, loc=<0 0 0>, rot=<0 0 0 1>}
 upperarm_l_ctrl {controlType=EULER_TRANSFORM, type=Control, loc=<0 0 0>, rot=<0 0 0 1>}
]
variables [
 isCrouching {type=bool, default=False}
 isCrouchingSprinting {type=bool, default=False}
 isSkyDiving {type=bool, default=False}
 isSliding {type=bool, default=False}
 isSlidingBaseBall {type=bool, default=False}
 isTargetStandingUp {type=bool, default=False}
]
