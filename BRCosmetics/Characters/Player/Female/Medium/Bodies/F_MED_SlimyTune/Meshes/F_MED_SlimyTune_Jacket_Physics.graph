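# Physics asset for the jacket mesh of the SlimyTune female medium player body, defining capsule bodies for jacket secondary motion.
# Seven two-link jacket chains hang from tiny base anchor capsules, with twist limits at zero and tight swing limits, some on one axis only.
# Larger capsules on the pelvis, thighs and right bag provide body shapes around the chains.
# Front and back strap bones also have capsules but no constraints here.
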PhysicsAsset /BRCosmetics/Characters/Player/Female/Medium/Bodies/F_MED_SlimyTune/Meshes/F_MED_SlimyTune_Jacket_Physics
bodies {
 base_bag_r [{
   shapeType=Capsule
   length=2.0229
   radius=11.9781
   shapeName=base_bag_r_capsule
   center=<2.9118 -0.6697 -0.769>
   rotation=<rot 90 -165.9638 -87.7849>
  }]
 base_jacket_0_l [{shapeType=Capsule, length=0.505, radius=0.505, shapeName=base_jacket_0_l_capsule, rotation=<rot 0 -90 0>}]
 base_jacket_0_m [{shapeType=Capsule, length=0.505, radius=0.505, shapeName=base_jacket_0_m_capsule, rotation=<rot 0 -90 0>}]
 base_jacket_0_r [{shapeType=Capsule, length=0.505, radius=0.505, shapeName=base_jacket_0_r_capsule, rotation=<rot 0 -90 0>}]
 base_jacket_1_l [{shapeType=Capsule, length=0.505, radius=0.505, shapeName=base_jacket_1_l_capsule, rotation=<rot 0 -90 0>}]
 base_jacket_1_r [{shapeType=Capsule, length=0.505, radius=0.505, shapeName=base_jacket_1_r_capsule, rotation=<rot 0 -90 0>}]
 base_jacket_2_l [{shapeType=Capsule, length=0.505, radius=0.505, shapeName=base_jacket_2_l_capsule, rotation=<rot 0 -90 0>}]
 base_jacket_2_r [{shapeType=Capsule, length=0.505, radius=0.505, shapeName=base_jacket_2_r_capsule, rotation=<rot 0 -90 0>}]
 dyn_jacket_0_end_l [{
   shapeType=Capsule
   length=12.4618
   radius=5.4167
   shapeName=dyn_jacket_0_end_l_capsule
   center=<15.8727 -1.079 -14.8257>
   rotation=<rot -4.4931 -8.0719 57.3106>
  }]
 dyn_jacket_0_end_m [{shapeType=Capsule, length=26.1627, radius=6.997, shapeName=dyn_jacket_0_end_m_capsule, center=<16.5422 0.6923 0>, rotation=<rot 0 14.6165 0>}]
 dyn_jacket_0_end_r [{
   shapeType=Capsule
   length=14.9197
   radius=2.9588
   shapeName=dyn_jacket_0_end_l_capsule
   center=<16.0655 -8.887 3.2824>
   rotation=<rot -11.2698 172.1908 -122.8687>
  }]
 dyn_jacket_0_l [
  {
   shapeType=Capsule
   length=12.4936
   radius=5.3583
   shapeName=dyn_jacket_0_l_capsule
   center=<13.2028 -6.7317 -0.9768>
   rotation=<rot -4.4931 -8.0719 57.3106>
  }
  {
   shapeType=Capsule
   length=12.4936
   radius=5.3583
   shapeName=dyn_jacket_0_l_capsule
   center=<20.507 -8.4561 -0.373>
   rotation=<rot -4.4931 -8.0719 57.3106>
  }
 ]
 dyn_jacket_0_m [{shapeType=Capsule, length=26.1627, radius=6.997, shapeName=dyn_jacket_0_m_capsule, center=<17.2703 -2.0994 0>, rotation=<rot 0 14.6165 0>}]
 dyn_jacket_0_r [
  {
   shapeType=Capsule
   length=14.7072
   radius=3.1447
   shapeName=dyn_jacket_0_l_capsule
   center=<14.008 -7.2557 0.4996>
   rotation=<rot -54.2775 -84.2589 167.7476>
  }
  {
   shapeType=Capsule
   length=14.7072
   radius=3.1447
   shapeName=dyn_jacket_0_l_capsule
   center=<21.3273 -8.0135 1.9096>
   rotation=<rot -55.6964 -83.8644 166.6369>
  }
  {shapeType=Capsule, length=15, radius=5, shapeName=dyn_jacket_0_r_capsule, center=<22.124 0 1.1702>, rotation=<rot 81.6387 180 -128.6751>}
 ]
 dyn_jacket_1_end_l [{
   shapeType=Capsule
   length=14.1363
   radius=5.2136
   shapeName=dyn_jacket_1_end_l_capsule
   center=<15.1941 0.1315 -0.0392>
   rotation=<rot -6.2882 -0.0007 -0.0017>
  }]
 dyn_jacket_1_end_r [{
   shapeType=Capsule
   length=14.1892
   radius=3.756
   shapeName=dyn_jacket_1_end_r_capsule
   center=<12.6391 6.4384 -0.205>
   rotation=<rot 5.7963 19.037 1.9958>
  }]
 dyn_jacket_1_l [{shapeType=Capsule, length=11.3869, radius=5.2136, shapeName=dyn_jacket_1_l_capsule, center=<13.6 -1.5811 0>}]
 dyn_jacket_1_r [{shapeType=Capsule, length=8.9585, radius=4.0415, shapeName=dyn_jacket_1_r_capsule, center=<22.3928 0.489 0>}]
 dyn_jacket_2_end_l [{shapeType=Capsule, length=10, radius=3, shapeName=dyn_jacket_2_end_l_capsule, center=<13 0.2854 0>}]
 dyn_jacket_2_end_r [{
   shapeType=Capsule
   length=10
   radius=3
   shapeName=dyn_jacket_2_end_r_capsule
   center=<13.6293 2.0099 0.6695>
   rotation=<rot 8.8592 2.5569 16.1699>
  }]
 dyn_jacket_2_l [{shapeType=Capsule, length=10, radius=3, shapeName=dyn_jacket_2_l_capsule, center=<13 1.666 0>}]
 dyn_jacket_2_r [{
   shapeType=Capsule
   length=10
   radius=3
   shapeName=dyn_jacket_2_r_capsule
   center=<13.0001 -0.0874 1.4998>
   rotation=<rot -0.0039 0.002 15.8493>
  }]
 dyn_strapBck_01 [{shapeType=Capsule, length=17.6779, radius=4.1654, shapeName=dyn_strapBck_01_capsule, center=<0 0 -0.6129>, rotation=<rot 90 0 0>}]
 dyn_strapBck_02 [{shapeType=Capsule, length=19.8872, radius=3.9311, shapeName=dyn_strapBck_02_capsule, rotation=<rot 90 0 0>}]
 dyn_strapFrnt_01 [{
   shapeType=Capsule
   length=16.8711
   radius=4.1289
   shapeName=dyn_strapFrnt_01_capsule
   center=<0.4651 0 -1.2226>
   rotation=<rot 88.3085 -0.001 -0.0011>
  }]
 dyn_strapFrnt_02 [{
   shapeType=Capsule
   length=20.5541
   radius=3.173
   shapeName=dyn_strapFrnt_02_capsule
   center=<-0.4381 -0.2396 0.3078>
   rotation=<rot 85.9197 -0.0011 -2.4322>
  }]
 pelvis [{shapeType=Capsule, length=9.4543, radius=10.5457, shapeName=pelvis_capsule, center=<-4.6598 -3.3795 0>}]
 thigh_l [{
   shapeType=Capsule
   length=34.4355
   radius=7.9313
   shapeName=thigh_l_capsule
   center=<-19.1274 -0.0949 -0.0051>
   rotation=<rot 87.2716 -177.5231 179.9999>
  }]
 thigh_r [{
   shapeType=Capsule
   length=33.7986
   radius=8.5683
   shapeName=thigh_l_capsule
   center=<19.3801 0.4434 -0.027>
   rotation=<rot 87.2255 -177.5433 179.9866>
  }]
}
constraints [
 {bone1Name=dyn_jacket_0_end_l, bone2Name=dyn_jacket_0_l, swing1LimitDegrees=20, swing2LimitDegrees=10, twistLimitDegrees=0}
 {bone1Name=dyn_jacket_0_l, bone2Name=base_jacket_0_l, swing1LimitDegrees=20, swing2LimitDegrees=10, twistLimitDegrees=0}
 {bone1Name=dyn_jacket_0_r, bone2Name=base_jacket_0_r, swing1LimitDegrees=20, swing2LimitDegrees=15, twistLimitDegrees=0}
 {bone1Name=dyn_jacket_0_end_r, bone2Name=dyn_jacket_0_r, swing1LimitDegrees=20, swing2LimitDegrees=10, twistLimitDegrees=0}
 {bone1Name=dyn_jacket_2_end_l, bone2Name=dyn_jacket_2_l, swing1LimitDegrees=15, swing2LimitDegrees=0, twistLimitDegrees=0}
 {bone1Name=dyn_jacket_2_l, bone2Name=base_jacket_2_l, swing1LimitDegrees=15, swing2LimitDegrees=0, twistLimitDegrees=0}
 {bone1Name=dyn_jacket_2_end_r, bone2Name=dyn_jacket_2_r, swing1LimitDegrees=25, swing2LimitDegrees=0, twistLimitDegrees=0}
 {bone1Name=dyn_jacket_2_r, bone2Name=base_jacket_2_r, swing1LimitDegrees=25, swing2LimitDegrees=0, twistLimitDegrees=0}
 {bone1Name=dyn_jacket_1_end_r, bone2Name=dyn_jacket_1_r, swing1LimitDegrees=10, swing2LimitDegrees=10, twistLimitDegrees=0}
 {bone1Name=dyn_jacket_1_r, bone2Name=base_jacket_1_r, swing1LimitDegrees=10, swing2LimitDegrees=10, twistLimitDegrees=0}
 {bone1Name=dyn_jacket_1_end_l, bone2Name=dyn_jacket_1_l, swing1LimitDegrees=10, swing2LimitDegrees=10, twistLimitDegrees=0}
 {bone1Name=dyn_jacket_1_l, bone2Name=base_jacket_1_l, swing1LimitDegrees=10, swing2LimitDegrees=10, twistLimitDegrees=0}
 {bone1Name=dyn_jacket_0_end_m, bone2Name=dyn_jacket_0_m, swing1LimitDegrees=30, swing2LimitDegrees=0, twistLimitDegrees=0}
 {bone1Name=dyn_jacket_0_m, bone2Name=base_jacket_0_m, swing1LimitDegrees=30, swing2LimitDegrees=0, twistLimitDegrees=0}
]
