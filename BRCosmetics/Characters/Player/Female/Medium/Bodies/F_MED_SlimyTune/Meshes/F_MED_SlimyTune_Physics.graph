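# Physics asset for the SlimyTune medium female player body, built around a hanging strap with front and back capsule chains.
# Each chain starts at a tiny base anchor with zero-degree root limits, then allows small swing and no twist along its links.
# The chain tips are joined by an angle-locked constraint to close the loop, and the front adds a rigid "fix" link near its root.
# Unconstrained pelvis and left-thigh capsules supply body shapes beside the strap.
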PhysicsAsset /BRCosmetics/Characters/Player/Female/Medium/Bodies/F_MED_SlimyTune/Meshes/F_MED_SlimyTune_Physics
bodies {
 base_strapBck_0 [{shapeType=Capsule, length=0.505, radius=0.505, shapeName=base_strapBck_0_capsule, rotation=<rot 0 -90 0>}]
 base_strapFrnt_0 [{shapeType=Capsule, length=0.505, radius=0.505, shapeName=base_strapFrnt_0_capsule, rotation=<rot 0 -90 0>}]
 dyn_strapBck_00 [{shapeType=Capsule, length=4.642, radius=2.9531, shapeName=dyn_strapBck_00_capsule, center=<2 0 0>, rotation=<rot 83.8238 -180 -180>}]
 dyn_strapBck_01 [{
   shapeType=Capsule
   length=5.3282
   radius=2.7727
   shapeName=dyn_strapBck_01_capsule
   center=<2.0394 0 0.5104>
   rotation=<rot 85.5826 -0.0001 -0.0001>
  }]
 dyn_strapBck_02 [{
   shapeType=Capsule
   length=6.124
   radius=2.8381
   shapeName=dyn_strapBck_02_capsule
   center=<0.9576 0 0.1218>
   rotation=<rot 83.3361 -0.0001 -0.0001>
  }]
 dyn_strapBck_03 [{shapeType=Capsule, length=6.2849, radius=2.3078, shapeName=dyn_strapBck_03_capsule, center=<2 0 0>, rotation=<rot 82.9548 -0.0446 -0.0459>}]
 dyn_strapBck_04 [{
   shapeType=Capsule
   length=4.5167
   radius=2
   shapeName=dyn_strapBck_04_capsule
   center=<1.9618 1.6826 -0.5893>
   rotation=<rot 90 180 178.6991>
  }]
 dyn_strapFrnt_00 [{
   shapeType=Capsule
   length=3.8249
   radius=3.3147
   shapeName=dyn_strapFrnt_00_capsule
   center=<5.0192 1.7734 1.546>
   rotation=<rot 76.2515 -179.9999 -180>
  }]
 dyn_strapFrnt_00_fix [{
   shapeType=Capsule
   length=3.9591
   radius=3.1805
   shapeName=dyn_strapFrnt_00_capsule
   center=<7.7374 1.5877 1.4365>
   rotation=<rot 87.6948 -0.0004 -0.0005>
  }]
 dyn_strapFrnt_01 [{shapeType=Capsule, length=1.9908, radius=3.0692, shapeName=dyn_strapFrnt_01_capsule, center=<10.9532 1.1419 0.9515>, rotation=<rot 90 0 0>}]
 dyn_strapFrnt_02 [{
   shapeType=Capsule
   length=4.2953
   radius=1.4875
   shapeName=dyn_strapFrnt_02_capsule
   center=<8.693 1.4892 0.7276>
   rotation=<rot 89.1719 -180 -180>
  }]
 dyn_strapFrnt_03 [{
   shapeType=Capsule
   length=4.4045
   radius=1.2874
   shapeName=dyn_strapFrnt_03_capsule
   center=<4.7638 1.6449 0.982>
   rotation=<rot 87.2821 -179.9447 -179.9494>
  }]
 dyn_strapFrnt_04 [{
   shapeType=Capsule
   length=4.8582
   radius=1.5749
   shapeName=dyn_strapFrnt_04_capsule
   center=<4.649 1.0815 -0.4437>
   rotation=<rot 43.6409 -0.3698 8.3841>
  }]
 pelvis [{shapeType=Capsule, length=17.9087, radius=10.2076, shapeName=pelvis_capsule, center=<-5.3791 0.6303 0>}]
 thigh_l [{shapeType=Capsule, length=32.6154, radius=7.505, shapeName=thigh_l_capsule, center=<-19.2677 -1.0081 0.1431>, rotation=<rot 89.7352 180 180>}]
}
constraints [
 {bone1Name=dyn_strapFrnt_00, bone2Name=base_strapFrnt_0, swing1LimitDegrees=0, swing2LimitDegrees=0, twistLimitDegrees=0}
 {bone1Name=dyn_strapFrnt_01, bone2Name=dyn_strapFrnt_00_fix, swing1LimitDegrees=10, swing2LimitDegrees=10, twistLimitDegrees=0}
 {bone1Name=dyn_strapFrnt_02, bone2Name=dyn_strapFrnt_01, swing1LimitDegrees=10, swing2LimitDegrees=10, twistLimitDegrees=0}
 {bone1Name=dyn_strapFrnt_03, bone2Name=dyn_strapFrnt_02, swing1LimitDegrees=10, swing2LimitDegrees=10, twistLimitDegrees=0}
 {bone1Name=dyn_strapFrnt_04, bone2Name=dyn_strapFrnt_03, swing1LimitDegrees=10, swing2LimitDegrees=10, twistLimitDegrees=0}
 {bone1Name=dyn_strapBck_00, bone2Name=base_strapBck_0, swing1LimitDegrees=0, swing2LimitDegrees=0, twistLimitDegrees=0}
 {bone1Name=dyn_strapBck_01, bone2Name=dyn_strapBck_00, swing1LimitDegrees=2, swing2LimitDegrees=2, twistLimitDegrees=0}
 {bone1Name=dyn_strapBck_02, bone2Name=dyn_strapBck_01, swing1LimitDegrees=10, swing2LimitDegrees=10, twistLimitDegrees=0}
 {bone1Name=dyn_strapBck_03, bone2Name=dyn_strapBck_02, swing1LimitDegrees=10, swing2LimitDegrees=10, twistLimitDegrees=0}
 {bone1Name=dyn_strapBck_04, bone2Name=dyn_strapBck_03, swing1LimitDegrees=10, swing2LimitDegrees=10, twistLimitDegrees=0}
 {
  bone1Name=dyn_strapBck_04
  bone2Name=dyn_strapFrnt_04
  swing1LimitDegrees=30
  swing1Motion=Locked
  swing2LimitDegrees=30
  swing2Motion=Locked
  twistLimitDegrees=20
  twistMotion=Locked
 }
 {bone1Name=dyn_strapFrnt_00_fix, bone2Name=dyn_strapFrnt_00, swing1LimitDegrees=0, swing2LimitDegrees=0, twistLimitDegrees=0}
]
