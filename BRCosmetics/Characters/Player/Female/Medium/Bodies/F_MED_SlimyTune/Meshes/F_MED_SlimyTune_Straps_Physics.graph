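# Physics asset for the chest straps of the SlimyTune female medium body cosmetic, giving left and right strap bones a physical setup for secondary motion.
# Each side chains three capsule bodies from a small base handle body, with constraints allowing narrow swing and zero twist.
# Larger capsules on two spine bones and both thighs have no constraints and appear to act as body colliders for the straps.
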PhysicsAsset /BRCosmetics/Characters/Player/Female/Medium/Bodies/F_MED_SlimyTune/Meshes/F_MED_SlimyTune_Straps_Physics
bodies {
 base_handle_chest_strap_l [{shapeType=Capsule, length=0.505, radius=0.505, shapeName=base_handle_chest_strap_l_capsule, rotation=<rot 0 -90 0>}]
 base_handle_chest_strap_r [{shapeType=Capsule, length=0.505, radius=0.505, shapeName=base_handle_chest_strap_r_capsule, rotation=<rot 0 -90 0>}]
 dyn_chest_strap0_l [{shapeType=Capsule, length=5.7377, radius=2, shapeName=dyn_chest_strap0_l_capsule, center=<4 0 1>, rotation=<rot 90 0 0>}]
 dyn_chest_strap0_r [{shapeType=Capsule, length=4.901, radius=2, shapeName=dyn_chest_strap0_r_capsule, center=<4 0 1>, rotation=<rot 90 0 0>}]
 dyn_chest_strap1_l [{shapeType=Capsule, length=6.8721, radius=2, shapeName=dyn_chest_strap1_l_capsule, center=<4 0 1>, rotation=<rot 90 0 0>}]
 dyn_chest_strap1_r [{shapeType=Capsule, length=7.064, radius=2.4098, shapeName=dyn_chest_strap1_r_capsule, center=<4 0 1>, rotation=<rot 90 0 0>}]
 dyn_chest_strap2_l [{shapeType=Capsule, length=7.2501, radius=2.5311, shapeName=dyn_chest_strap2_l_capsule, center=<0.5168 0 1>, rotation=<rot 90 0 0>}]
 dyn_chest_strap2_r [{shapeType=Capsule, length=3.4358, radius=2.7525, shapeName=dyn_chest_strap2_r_capsule, center=<4 0 1>, rotation=<rot 90 175.632 -96.2752>}]
 spine_02 [{shapeType=Capsule, length=14.7242, radius=10.4865, shapeName=spine_02_capsule, center=<-0.1472 1.0625 0>}]
 spine_04 [{shapeType=Capsule, length=12.8828, radius=10.1986, shapeName=spine_04_capsule, center=<-2.303 2.9935 0>}]
 thigh_l [{
   shapeType=Capsule
   length=34.4355
   radius=7.9313
   shapeName=thigh_l_capsule
   center=<-19.1274 -0.0949 -0.0051>
   rotation=<rot 87.2716 -177.5231 179.9999>
  }]
 thigh_r [{
   shapeType=Capsule
   length=33.7986
   radius=8.5683
   shapeName=thigh_l_capsule
   center=<19.3801 0.4434 -0.027>
   rotation=<rot 87.2255 -177.5433 179.9866>
  }]
}
constraints [
 {bone1Name=dyn_chest_strap0_l, bone2Name=base_handle_chest_strap_l, swing1LimitDegrees=20, swing2LimitDegrees=20, twistLimitDegrees=0}
 {bone1Name=dyn_chest_strap1_l, bone2Name=dyn_chest_strap0_l, swing1LimitDegrees=20, swing2LimitDegrees=20, twistLimitDegrees=0}
 {bone1Name=dyn_chest_strap2_l, bone2Name=dyn_chest_strap1_l, swing1LimitDegrees=20, swing2LimitDegrees=20, twistLimitDegrees=0}
 {bone1Name=dyn_chest_strap0_r, bone2Name=base_handle_chest_strap_r, swing1LimitDegrees=20, swing2LimitDegrees=20, twistLimitDegrees=0}
 {bone1Name=dyn_chest_strap1_r, bone2Name=dyn_chest_strap0_r, swing1LimitDegrees=20, swing2LimitDegrees=20, twistLimitDegrees=0}
 {bone1Name=dyn_chest_strap2_r, bone2Name=dyn_chest_strap1_r, swing1LimitDegrees=20, swing2LimitDegrees=20, twistLimitDegrees=0}
]
