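# Animation Blueprint for the F_MED_SlimyTune player cosmetic body, adding secondary motion to its jacket, straps, bag and zipper.
# It copies body and layered face poses, then chains cached stages through a Control Rig, weighted bone constraints and additive jacket and strap correctives.
# RigidBody simulations interleave with these correctives, and AnimDynamics hip straps precede a final strap simulation.
# The event graph derives movement, crouch, jump, gravity and backbling state.
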AnimBlueprint /BRCosmetics/Characters/Player/Female/Medium/Bodies/F_MED_SlimyTune/Meshes/F_MED_SlimyTune_AnimBP
target_skeleton=&/F_MED_SlimyTune_Skeleton
graphs [
 AnimGraph {nodes [
   n1 = RigidBody!(ComponentPose=Local To Component(LocalPose=Use cached pose 'HipStrapsCache'()), Alpha, OverrideWorldGravity:Vector="0,0,0", ExternalForce) -> (Pose) {
    bClampLinearTranslationLimitToRefPose=true
    bOverrideWorldGravity=true
    overridePhysicsAsset=&/F_MED_SlimyTune_Straps_Physics
    simulationSpace=BaseBoneSpace
    alphaBoolBlend {blendOption=Linear}
    alphaScaleBiasClamp {interpSpeedDecreasing=0.1, interpSpeedIncreasing=1}
    baseBoneRef {boneName=spine_01}
    componentAppliedLinearAccClamp=<3000 3000 3000>
    componentLinearAccScale=<0.2 0.2 0.2>
    componentLinearVelScale=<0.2 0.2 0.2>
   }
   Output Pose(Result=Component To Local(ComponentPose=@n1)) {layerGroup=None}
   n2 = "Get Body Mesh "() -> ("Body Mesh ":Skeletal Mesh Component Object Reference)
   n3 = Layered blend per bone(
    BasePose=Layered blend per bone(
     BasePose=Copy Pose From Mesh() {bUseAttachedParent=true}
     BlendPoses_0=Copy Pose From Mesh(SourceMeshComponent=@n2) {bUseAttachedParent=true}
    )
    BlendPoses_0=Copy Pose From Mesh(SourceMeshComponent="Get Face Mesh "() -> ("Face Mesh ":Skeletal Mesh Component Object Reference)) {bUseAttachedParent=true}
   ) {layerSetup [{branchFilters [
       {boneName=FACIAL_C_FacialRoot}
       {boneName=FACIAL_C_UnderChin}
       {boneName=FACIAL_L_UnderChin}
       {boneName=FACIAL_R_UnderChin}
       {boneName=FACIAL_C_Jawline}
       {boneName=FACIAL_L_Jawline}
       {boneName=FACIAL_R_Jawline}
      ]}]}
   Save cached pose 'Start_Cache'(Pose=@n3) {cacheName=Start_Cache}
   n4 = Control Rig!(
    Source=Use cached pose 'Inputpose'()
    Alpha
    isCrouching:bool=Get bIsCrouching()
    isSliding:bool=Get bIsSliding()
    isCrouchingSprinting:bool=Get bIsCrouchSprinting()
    isSlidingBaseBall:bool=Get bIsBaseballSliding()
    isSkyDiving:bool=Get bIsSkyDiving()
    isTargetStandingUp:bool=Get targetStandingUp() -> (targetStandingUp:bool)
   ) -> (Pose) {
    alphaBoolBlend {blendInTime=1, blendOption=Linear, blendOutTime=1}
    controlRigAssetReference {blueprintRigClass=&/F_MED_SlimyTune_CtrlRig.F_MED_SlimyTune_CtrlRig_C}
   }
   Save cached pose 'ControlRigCache'(Pose=@n4) {cacheName=ControlRigCache}
   n5 = "Constraint - base_handle_chest_strap_l (PPPPPPP)"!(
    ComponentPose=Local To Component(LocalPose=Use cached pose 'ControlRigCache'())
    Alpha=1
    ConstraintWeights_0=0.371
    ConstraintWeights_1:float=0.2782
    ConstraintWeights_2:float=0.2675
    ConstraintWeights_3:float=0.0685
    ConstraintWeights_4:float=0.0106
    ConstraintWeights_5:float=0.0033
    ConstraintWeights_6:float=0.0009
   ) -> (Pose) {
    alphaBoolBlend {blendOption=Linear}
    boneToModify {boneName=base_handle_chest_strap_l}
    constraintSetup [
     {targetBone {boneName=spine_05}}
     {targetBone {boneName=spine_04}}
     {targetBone {boneName=deform_pec_l}}
     {targetBone {boneName=clavicle_l}}
     {targetBone {boneName=upperarm_twist_01_l}}
     {targetBone {boneName=deform_upperarm_fr_l}}
     {targetBone {boneName=deform_upperarm_up_l}}
    ]
    constraintWeights [1,1,1,1,1,1,1]
   }
   n6 = "Constraint - base_handle_chest_strap_r (PPPPPPP)"!(
    ComponentPose=@n5
    Alpha=1
    ConstraintWeights_0=0.371
    ConstraintWeights_1:float=0.2782
    ConstraintWeights_2:float=0.2675
    ConstraintWeights_3:float=0.0685
    ConstraintWeights_4:float=0.0106
    ConstraintWeights_5:float=0.0033
    ConstraintWeights_6:float=0.0009
   ) -> (Pose) {
    alphaBoolBlend {blendOption=Linear}
    boneToModify {boneName=base_handle_chest_strap_r}
    constraintSetup [
     {targetBone {boneName=spine_05}}
     {targetBone {boneName=spine_04}}
     {targetBone {boneName=deform_pec_r}}
     {targetBone {boneName=clavicle_r}}
     {targetBone {boneName=upperarm_twist_01_r}}
     {targetBone {boneName=deform_upperarm_fr_r}}
     {targetBone {boneName=deform_upperarm_up_r}}
    ]
    constraintWeights [1,1,1,1,1,1,1]
   }
   n7 = "Constraint - base_hip_strap1_l (PPPP)"!(
    ComponentPose=@n6
    Alpha=1
    ConstraintWeights_0=0.6409
    ConstraintWeights_1:float=0.3207
    ConstraintWeights_2:float=0.0363
    ConstraintWeights_3:float=0.0021
   ) -> (Pose) {
    alphaBoolBlend {blendOption=Linear}
    boneToModify {boneName=base_hip_strap1_l}
    constraintSetup [
     {targetBone {boneName=dyn_jacket_0_m}}
     {targetBone {boneName=dyn_jacket_2_l}}
     {targetBone {boneName=dyn_jacket_2_r}}
     {targetBone {boneName=dyn_jacket_1_l}}
    ]
    constraintWeights [1,1,1,1]
   }
   n8 = "Constraint - base_hip_strap1_r (PPPP)"!(
    ComponentPose=@n7
    Alpha=1
    ConstraintWeights_0=0.6358
    ConstraintWeights_1:float=0.3258
    ConstraintWeights_2:float=0.0361
    ConstraintWeights_3:float=0.0023
   ) -> (Pose) {
    alphaBoolBlend {blendOption=Linear}
    boneToModify {boneName=base_hip_strap1_r}
    constraintSetup [
     {targetBone {boneName=dyn_jacket_0_m}}
     {targetBone {boneName=dyn_jacket_2_r}}
     {targetBone {boneName=dyn_jacket_2_l}}
     {targetBone {boneName=dyn_jacket_1_r}}
    ]
    constraintWeights [1,1,1,1]
   }
   n9 = "Constraint - base_bag_r (PP)"!(ComponentPose=@n8, Alpha=1, ConstraintWeights_0=0.7879, ConstraintWeights_1:float=0.2121) -> (Pose) {
    alphaBoolBlend {blendOption=Linear}
    boneToModify {boneName=base_bag_r}
    constraintSetup [{targetBone {boneName=thigh_r}}, {targetBone {boneName=thigh_twist_01_r}}]
    constraintWeights [1,1]
   }
   Save cached pose 'SurfaceConstraintsCache'(Pose=Component To Local(ComponentPose=@n9)) {cacheName=SurfaceConstraintsCache}
   n10 = "Transform (Modify) Bone - Bone: base_jacket_0_r"!(
    ComponentPose=Local To Component(LocalPose=Use cached pose 'SurfaceConstraintsCache'())
    bAlphaBoolEnabled:bool=True
    Translation="0,-20,0"
    Rotation="9.271,-18.324,115.928"
   ) -> (Pose) {
    alphaInputType=Bool
    rotationMode=BMM_Additive
    rotationSpace=BCS_BoneSpace
    translationMode=BMM_Additive
    translationSpace=BCS_BoneSpace
    alphaBoolBlend {blendInTime=1, blendOption=Linear, blendOutTime=1}
    boneToModify {boneName=base_jacket_0_r}
    rotation=<rot 9.271 -18.324 115.928>
    translation {y=-20}
   }
   n11 = "Transform (Modify) Bone - Bone: base_jacket_0_l"!(ComponentPose=@n10, bAlphaBoolEnabled:bool=True, Translation="5.316,-6.137,2.242", Rotation="-33.213,47.356,-30.743") -> (Pose) {
    alphaInputType=Bool
    rotationMode=BMM_Additive
    rotationSpace=BCS_BoneSpace
    translationMode=BMM_Additive
    translationSpace=BCS_BoneSpace
    alphaBoolBlend {blendInTime=1, blendOption=Linear, blendOutTime=1}
    boneToModify {boneName=base_jacket_0_l}
    rotation=<rot -33.213 47.356 -30.743>
    translation=<5.316 -6.137 2.242>
   }
   n12 = "Transform (Modify) Bone - Bone: base_strapFrnt_0"!(ComponentPose=@n11, bAlphaBoolEnabled:bool=True, Translation="0.187,-0.141,1.333", Rotation="-9.911,-37.020,-0.000") -> (Pose) {
    alphaInputType=Bool
    rotationMode=BMM_Additive
    rotationSpace=BCS_BoneSpace
    translationMode=BMM_Additive
    translationSpace=BCS_BoneSpace
    alphaBoolBlend {blendInTime=1, blendOption=Linear, blendOutTime=1}
    boneToModify {boneName=base_strapFrnt_0}
    rotation {pitch=-9.911, yaw=-37.02}
    translation=<0.187 -0.141 1.333>
   }
   n13 = "Transform (Modify) Bone - Bone: base_strapBck_0"!(ComponentPose=@n12, bAlphaBoolEnabled:bool=True, Translation="0,0,0", Rotation="-81.265,58.307,-49.963") -> (Pose) {
    alphaInputType=Bool
    rotationMode=BMM_Additive
    rotationSpace=BCS_BoneSpace
    translationMode=BMM_Additive
    translationSpace=BCS_BoneSpace
    alphaBoolBlend {blendInTime=1, blendOption=Linear, blendOutTime=1}
    boneToModify {boneName=base_strapBck_0}
    rotation=<rot -81.265 58.307 -49.963>
   }
   Save cached pose 'FixingOnGrindRailCache'(Pose=Component To Local(ComponentPose=@n13)) {cacheName=FixingOnGrindRailCache}
   n14 = "Transform (Modify) Bone - Bone: base_jacket_0_r"!(ComponentPose=Local To Component(LocalPose=Use cached pose 'FixingOnGrindRailCache'()), bAlphaBoolEnabled=True, Translation="0,0,-8") -> (Pose) {
    alphaInputType=Bool
    translationMode=BMM_Additive
    translationSpace=BCS_BoneSpace
    alphaBoolBlend {blendInTime=0.1, blendOption=Linear, blendOutTime=0.1}
    boneToModify {boneName=base_jacket_0_r}
    translation {z=-8}
   }
   Save cached pose 'CrouchSprintCache'(Pose=Component To Local(ComponentPose=@n14)) {cacheName=CrouchSprintCache}
   n15 = "Transform (Modify) Bone - Bone: base_jacket_0_l"(
    ComponentPose=Local To Component(LocalPose=Use cached pose 'CrouchSprintCache'())
    bAlphaBoolEnabled=True
    Translation="0,0,0"
    Rotation="22.762,8.712,3.384"
   ) {
    alphaInputType=Bool
    rotationMode=BMM_Additive
    rotationSpace=BCS_BoneSpace
    translationMode=BMM_Additive
    translationSpace=BCS_BoneSpace
    alphaBoolBlend {blendOption=Linear}
    alphaScaleBiasClamp {bInterpResult=true, interpSpeedDecreasing=5, interpSpeedIncreasing=5}
    boneToModify {boneName=base_jacket_0_l}
    rotation=<rot 22.762 8.712 3.384>
   }
   n16 = "Transform (Modify) Bone - Bone: base_jacket_0_r"!(ComponentPose=@n15, bAlphaBoolEnabled=True, Translation="0,-8,0", Rotation:Rotator="0,0,0") -> (Pose) {
    alphaInputType=Bool
    rotationMode=BMM_Additive
    rotationSpace=BCS_BoneSpace
    translationMode=BMM_Additive
    translationSpace=BCS_BoneSpace
    alphaBoolBlend {blendOption=Linear}
    alphaScaleBiasClamp {bInterpResult=true, interpSpeedDecreasing=5, interpSpeedIncreasing=5}
    boneToModify {boneName=base_jacket_0_r}
    translation {y=-8}
   }
   n17 = "Transform (Modify) Bone - Bone: base_jacket_0_l"(ComponentPose=@n16, bAlphaBoolEnabled=True, Translation="3.559,0.646,1.448", Rotation="74.239,-9.471,-16.370") {
    alphaInputType=Bool
    rotationMode=BMM_Additive
    rotationSpace=BCS_BoneSpace
    translationMode=BMM_Additive
    translationSpace=BCS_BoneSpace
    alphaBoolBlend {blendOption=Linear}
    alphaScaleBiasClamp {bInterpResult=true, interpSpeedDecreasing=5, interpSpeedIncreasing=5}
    boneToModify {boneName=base_jacket_0_l}
    rotation=<rot 74.239 -9.471 -16.37>
    translation=<3.559 0.646 1.448>
   }
   n18 = "Transform (Modify) Bone - Bone: base_chest_strap0_l"!(ComponentPose=@n17, bAlphaBoolEnabled:bool=True, Translation="0,0,0", Rotation="0,30,0") -> (Pose) {
    alphaInputType=Bool
    rotationMode=BMM_Additive
    rotationSpace=BCS_BoneSpace
    translationMode=BMM_Additive
    translationSpace=BCS_BoneSpace
    alphaBoolBlend {blendOption=Linear}
    alphaScaleBiasClamp {bInterpResult=true, interpSpeedDecreasing=5, interpSpeedIncreasing=5}
    boneToModify {boneName=base_chest_strap0_l}
    rotation {yaw=30}
   }
   n19 = "Transform (Modify) Bone - Bone: base_jacket_0_r"!(ComponentPose=@n18, Alpha:float=1, Translation="2,-5.8,-4.65", Rotation="0,0,0") -> (Pose) {
    rotationMode=BMM_Additive
    rotationSpace=BCS_BoneSpace
    translationMode=BMM_Additive
    translationSpace=BCS_BoneSpace
    alphaBoolBlend {blendOption=Linear}
    alphaScaleBiasClamp {bInterpResult=true, interpSpeedDecreasing=5, interpSpeedIncreasing=5}
    boneToModify {boneName=base_jacket_0_r}
    translation=<2 -5.8 -4.65>
   }
   n20 = "Transform (Modify) Bone - Bone: base_jacket_0_r"(ComponentPose=@n19, Alpha=1, Translation="0,0,0", Rotation="10,-10,0") {
    rotationMode=BMM_Additive
    rotationSpace=BCS_BoneSpace
    translationMode=BMM_Additive
    translationSpace=BCS_BoneSpace
    alphaBoolBlend {blendOption=Linear}
    alphaScaleBiasClamp {bInterpResult=true, interpSpeedDecreasing=5, interpSpeedIncreasing=5}
    boneToModify {boneName=base_jacket_0_r}
    rotation {pitch=10, yaw=-10}
   }
   n21 = "Transform (Modify) Bone - Bone: base_jacket_0_l"!(ComponentPose=@n20, Alpha:float=1, Translation="0,0,0", Rotation="15,-15,0") -> (Pose) {
    rotationMode=BMM_Additive
    rotationSpace=BCS_BoneSpace
    translationMode=BMM_Additive
    translationSpace=BCS_BoneSpace
    alphaBoolBlend {blendOption=Linear}
    alphaScaleBiasClamp {bInterpResult=true, interpSpeedDecreasing=5, interpSpeedIncreasing=5}
    boneToModify {boneName=base_jacket_0_l}
    rotation {pitch=15, yaw=-15}
   }
   n22 = "Transform (Modify) Bone - Bone: base_jacket_0_l"(ComponentPose=@n21, Alpha=1, Translation="0,0,0", Rotation="60,-60,0") {
    rotationMode=BMM_Additive
    rotationSpace=BCS_BoneSpace
    translationMode=BMM_Additive
    translationSpace=BCS_BoneSpace
    alphaBoolBlend {blendOption=Linear}
    alphaScaleBiasClamp {bInterpResult=true, interpSpeedDecreasing=4, interpSpeedIncreasing=4}
    boneToModify {boneName=base_jacket_0_l}
    rotation {pitch=60, yaw=-60}
   }
   n23 = "Transform (Modify) Bone - Bone: base_jacket_0_r"(ComponentPose=@n22, Alpha=1, Translation="6.417,-1.247,-0.154", Rotation="-5.583,-23.189,15.058") {
    rotationMode=BMM_Additive
    rotationSpace=BCS_BoneSpace
    translationMode=BMM_Additive
    translationSpace=BCS_BoneSpace
    alphaBoolBlend {blendOption=Linear}
    alphaScaleBiasClamp {bInterpResult=true, interpSpeedDecreasing=5, interpSpeedIncreasing=5}
    boneToModify {boneName=base_jacket_0_r}
    rotation=<rot -5.583 -23.189 15.058>
    translation=<6.417 -1.247 -0.154>
   }
   n24 = "Transform (Modify) Bone - Bone: base_jacket_0_r"!(ComponentPose=@n23, bAlphaBoolEnabled:bool=True, Translation="0,0,0", Rotation="-30,-30,0") -> (Pose) {
    alphaInputType=Bool
    rotationMode=BMM_Additive
    rotationSpace=BCS_BoneSpace
    translationMode=BMM_Additive
    translationSpace=BCS_BoneSpace
    alphaBoolBlend {blendOption=Linear}
    alphaScaleBiasClamp {bInterpResult=true, interpSpeedDecreasing=5, interpSpeedIncreasing=5}
    boneToModify {boneName=base_jacket_0_r}
    rotation {pitch=-30, yaw=-30}
   }
   n25 = "Transform (Modify) Bone - Bone: base_jacket_0_l"!(ComponentPose=@n24, bAlphaBoolEnabled:bool=True, Translation="0,0,0", Rotation="30,-40,0") -> (Pose) {
    alphaInputType=Bool
    rotationMode=BMM_Additive
    rotationSpace=BCS_BoneSpace
    translationMode=BMM_Additive
    translationSpace=BCS_BoneSpace
    alphaBoolBlend {blendInTime=1, blendOption=Linear, blendOutTime=1}
    alphaScaleBiasClamp {bInterpResult=true, interpSpeedDecreasing=5, interpSpeedIncreasing=5}
    boneToModify {boneName=base_jacket_0_l}
    rotation {pitch=30, yaw=-40}
   }
   n26 = "Transform (Modify) Bone - Bone: base_jacket_0_l"(ComponentPose=@n25, bAlphaBoolEnabled=True, Translation="0,-10,0", Rotation="0,-10,0") {
    alphaInputType=Bool
    rotationMode=BMM_Additive
    rotationSpace=BCS_BoneSpace
    translationMode=BMM_Additive
    translationSpace=BCS_BoneSpace
    alphaBoolBlend {blendInTime=5, blendOption=Linear, blendOutTime=5}
    alphaScaleBiasClamp {bInterpResult=true, interpSpeedDecreasing=5, interpSpeedIncreasing=5}
    boneToModify {boneName=base_jacket_0_l}
    rotation {yaw=-10}
    translation {y=-10}
   }
   n27 = "Transform (Modify) Bone - Bone: base_strapFrnt_0"!(ComponentPose=@n26, Alpha:float=1, Translation="0,0,3", Rotation="0,0,0") -> (Pose) {
    rotationMode=BMM_Additive
    rotationSpace=BCS_BoneSpace
    translationMode=BMM_Additive
    translationSpace=BCS_BoneSpace
    alphaBoolBlend {blendInTime=5, blendOption=Linear, blendOutTime=5}
    alphaScaleBiasClamp {bInterpResult=true, interpSpeedDecreasing=5, interpSpeedIncreasing=5}
    boneToModify {boneName=base_strapFrnt_0}
    translation {z=3}
   }
   n28 = RigidBody(ComponentPose=@n27, OverrideWorldGravity="0,0,0") {
    bClampLinearTranslationLimitToRefPose=true
    bOverrideWorldGravity=true
    overridePhysicsAsset=&/F_MED_SlimyTune_Physics
    simulationSpace=BaseBoneSpace
    alphaBoolBlend {blendOption=Linear}
    alphaScaleBiasClamp {bInterpResult=true}
    baseBoneRef {boneName=thigh_l}
    componentAppliedLinearAccClamp=<1000 1000 1000>
    componentLinearAccScale=<0.15 0.15 0.15>
    componentLinearVelScale=<0.15 0.15 0.15>
   }
   n29 = "Transform (Modify) Bone - Bone: base_jacket_0_l"(ComponentPose=@n28, bAlphaBoolEnabled=True, Translation="0,-8,0", Rotation="0,0,0") {
    alphaInputType=Bool
    rotationMode=BMM_Additive
    rotationSpace=BCS_BoneSpace
    translationMode=BMM_Additive
    translationSpace=BCS_BoneSpace
    alphaBoolBlend {blendInTime=0.1, blendOption=Linear, blendOutTime=0.1}
    boneToModify {boneName=base_jacket_0_l}
    translation {y=-8}
   }
   n30 = RigidBody(ComponentPose=@n29, OverrideWorldGravity="0,0,0") {
    bClampLinearTranslationLimitToRefPose=true
    bOverrideWorldGravity=true
    overridePhysicsAsset=&/F_MED_SlimyTune_Jacket_Physics
    simulationSpace=BaseBoneSpace
    alphaBoolBlend {blendOption=Linear}
    alphaScaleBiasClamp {bInterpResult=true}
    baseBoneRef {boneName=spine_03}
    componentAppliedLinearAccClamp=<3000 3000 3000>
    componentLinearAccScale=<0.4 0.4 0.4>
    componentLinearVelScale=<0.4 0.4 0.4>
    simSpaceSettings {
     maxAngularAcceleration=1000
     maxAngularVelocity=3
     maxLinearAcceleration=700
     maxLinearVelocity=500
     velocityScaleZ=0
     externalLinearDragV=<0.2 0.2>
    }
   }
   n31 = "Transform (Modify) Bone - Bone: dyn_jacket_0_l"!(ComponentPose=@n30, bAlphaBoolEnabled:bool=True, Translation="-0.720,-1.916,-7.948", Rotation="-2.693,-11.084,2.480", Scale="1,1,1") -> (Pose) {
    alphaInputType=Bool
    rotationMode=BMM_Additive
    rotationSpace=BCS_BoneSpace
    translationMode=BMM_Additive
    translationSpace=BCS_BoneSpace
    alphaBoolBlend {blendInTime=0.25, blendOption=Linear, blendOutTime=0.25}
    boneToModify {boneName=dyn_jacket_0_l}
    rotation=<rot -2.693 -11.084 2.48>
    translation=<-0.72 -1.916 -7.948>
   }
   n32 = "Transform (Modify) Bone - Bone: base_jacket_0_r"!(ComponentPose=@n31, bAlphaBoolEnabled=True, Translation="1.765,-4.265,-0.198", Rotation="-8.538,32.162,-1.569", Scale:Vector="1,1,1") -> (Pose) {
    alphaInputType=Bool
    rotationMode=BMM_Additive
    rotationSpace=BCS_BoneSpace
    translationMode=BMM_Additive
    translationSpace=BCS_BoneSpace
    alphaBoolBlend {blendInTime=1, blendOption=Linear, blendOutTime=1}
    boneToModify {boneName=base_jacket_0_r}
    rotation=<rot -8.538 32.162 -1.569>
    translation=<1.765 -4.265 -0.198>
   }
   n33 = "Transform (Modify) Bone - Bone: base_jacket_0_r"(ComponentPose=@n32, bAlphaBoolEnabled=True, Translation="-4.035,-3.850,5.348", Rotation="17.052,-18.169,-7.875", Scale="1,1,1") {
    alphaInputType=Bool
    rotationMode=BMM_Additive
    rotationSpace=BCS_BoneSpace
    translationMode=BMM_Additive
    translationSpace=BCS_BoneSpace
    alphaBoolBlend {blendInTime=1, blendOption=Linear, blendOutTime=1}
    boneToModify {boneName=base_jacket_0_r}
    rotation=<rot 17.052 -18.169 -7.875>
    translation=<-4.035 -3.85 5.348>
   }
   n34 = "Transform (Modify) Bone - Bone: base_jacket_0_l"!(ComponentPose=@n33, bAlphaBoolEnabled=True, Translation="-3.332,-4.557,2.544", Rotation="-2.853,-1.043,-29.745", Scale:Vector="1,1,1") -> (Pose) {
    alphaInputType=Bool
    rotationMode=BMM_Additive
    rotationSpace=BCS_BoneSpace
    translationMode=BMM_Additive
    translationSpace=BCS_BoneSpace
    alphaBoolBlend {blendInTime=1, blendOption=Linear, blendOutTime=1}
    boneToModify {boneName=base_jacket_0_l}
    rotation=<rot -2.853 -1.043 -29.745>
    translation=<-3.332 -4.557 2.544>
   }
   n35 = "Transform (Modify) Bone - Bone: dyn_jacket_0_end_l"!(ComponentPose=@n34, bAlphaBoolEnabled:bool=True, Translation="0,0,0", Rotation="-12.119,-0.036,0.021", Scale="1,1,1") -> (Pose) {
    alphaInputType=Bool
    rotationMode=BMM_Additive
    rotationSpace=BCS_BoneSpace
    translationMode=BMM_Additive
    translationSpace=BCS_BoneSpace
    alphaBoolBlend {blendInTime=1, blendOption=Linear, blendOutTime=1}
    boneToModify {boneName=dyn_jacket_0_end_l}
    rotation=<rot -12.119 -0.036 0.021>
   }
   Save cached pose 'DynLegStrapsJacketCache'(Pose=Component To Local(ComponentPose=@n35)) {cacheName=DynLegStrapsJacketCache}
   n36 = "Transform (Modify) Bone - Bone: base_jacket_0_l"!(
    ComponentPose=Local To Component(LocalPose=Use cached pose 'DynLegStrapsJacketCache'())
    Alpha:float=1
    Translation="-2.959,4.928,-6.004"
    Rotation="32.175,-16.101,-8.249"
   ) -> (Pose) {
    rotationMode=BMM_Additive
    rotationSpace=BCS_BoneSpace
    translationMode=BMM_Additive
    translationSpace=BCS_BoneSpace
    alphaBoolBlend {blendOption=Linear}
    alphaScaleBiasClamp {bInterpResult=true, interpSpeedDecreasing=1, interpSpeedIncreasing=1}
    boneToModify {boneName=base_jacket_0_l}
    rotation=<rot 32.175 -16.101 -8.249>
    translation=<-2.959 4.928 -6.004>
   }
   n37 = "Transform (Modify) Bone - Bone: dyn_jacket_0_end_l"!(ComponentPose=@n36, bAlphaBoolEnabled=True, Translation="1.5,-1,0.7", Rotation="-56.3,23.6,30") -> (Pose) {
    alphaInputType=Bool
    rotationMode=BMM_Additive
    rotationSpace=BCS_BoneSpace
    translationMode=BMM_Additive
    translationSpace=BCS_BoneSpace
    alphaBoolBlend {blendOption=Linear}
    boneToModify {boneName=dyn_jacket_0_end_l}
    rotation=<rot -56.3 23.6 30>
    translation=<1.5 -1 0.7>
   }
   n38 = "Transform (Modify) Bone - Bone: dyn_jacket_0_l"!(ComponentPose=@n37, bAlphaBoolEnabled=True, Translation="-4.808,-0.419,-3.376", Rotation="0.000,-9.500,-0.000") -> (Pose) {
    alphaInputType=Bool
    rotationMode=BMM_Additive
    rotationSpace=BCS_BoneSpace
    translationMode=BMM_Additive
    translationSpace=BCS_BoneSpace
    alphaBoolBlend {blendInTime=1, blendOption=Linear, blendOutTime=1}
    boneToModify {boneName=dyn_jacket_0_l}
    rotation {yaw=-9.5}
    translation=<-4.808 -0.419 -3.376>
   }
   n39 = "Transform (Modify) Bone - Bone: dyn_jacket_0_end_l"(ComponentPose=@n38, bAlphaBoolEnabled=True, Translation="2.529,2.478,0.000", Rotation="0.000,-45.582,-30.373") {
    alphaInputType=Bool
    rotationMode=BMM_Additive
    rotationSpace=BCS_BoneSpace
    translationMode=BMM_Additive
    translationSpace=BCS_BoneSpace
    alphaBoolBlend {blendInTime=1, blendOption=Linear, blendOutTime=1}
    boneToModify {boneName=dyn_jacket_0_end_l}
    rotation {roll=-30.373, yaw=-45.582}
    translation=<2.529 2.478>
   }
   Save cached pose 'fixAftrDynCache'(Pose=Component To Local(ComponentPose=@n39)) {cacheName=fixAftrDynCache}
   n40 = "Constraint - base_main_zipper (PPPPPPPP)"!(
    ComponentPose=Local To Component(LocalPose=Use cached pose 'fixAftrDynCache'())
    Alpha=1
    ConstraintWeights_0=0.469
    ConstraintWeights_1:float=0.1959
    ConstraintWeights_2:float=0.1331
    ConstraintWeights_3:float=0.104
    ConstraintWeights_4:float=0.0582
    ConstraintWeights_5:float=0.0218
    ConstraintWeights_6:float=0.0123
    ConstraintWeights_7:float=0.0056
   ) -> (Pose) {
    alphaBoolBlend {blendOption=Linear}
    boneToModify {boneName=base_main_zipper}
    constraintSetup [
     {targetBone {boneName=dyn_jacket_0_end_l}}
     {targetBone {boneName=dyn_jacket_1_end_l}}
     {targetBone {boneName=dyn_jacket_0_l}}
     {targetBone {boneName=dyn_jacket_1_l}}
     {targetBone {boneName=dyn_jacket_2_end_l}}
     {targetBone {boneName=dyn_jacket_2_l}}
     {targetBone {boneName=spine_04}}
     {targetBone {boneName=dyn_jacket_0_end_m}}
    ]
    constraintWeights [1,1,1,1,1,1,1,1]
   }
   n41 = "Constraint - base_hip_strap1_r (PPPPPPPP)"!(
    ComponentPose=@n40
    Alpha=1
    ConstraintWeights_0=0.2323
    ConstraintWeights_1:float=0.2042
    ConstraintWeights_2:float=0.1687
    ConstraintWeights_3:float=0.1567
    ConstraintWeights_4:float=0.1063
    ConstraintWeights_5:float=0.1063
    ConstraintWeights_6:float=0.016
    ConstraintWeights_7:float=0.0095
   ) -> (Pose) {
    alphaBoolBlend {blendOption=Linear}
    boneToModify {boneName=base_hip_strap1_r}
    constraintSetup [
     {targetBone {boneName=dyn_jacket_0_m}}
     {targetBone {boneName=dyn_jacket_2_r}}
     {targetBone {boneName=dyn_jacket_0_end_m}}
     {targetBone {boneName=dyn_jacket_2_end_r}}
     {targetBone {boneName=dyn_jacket_2_end_l}}
     {targetBone {boneName=dyn_jacket_2_l}}
     {targetBone {boneName=spine_04}}
     {targetBone {boneName=dyn_jacket_1_end_r}}
    ]
    constraintWeights [1,1,1,1,1,1,1,1]
   }
   n42 = "Constraint - base_hip_strap1_l (PPPPPPPP)"!(
    ComponentPose=@n41
    Alpha=1
    ConstraintWeights_0=0.2349
    ConstraintWeights_1:float=0.2102
    ConstraintWeights_2:float=0.1668
    ConstraintWeights_3:float=0.1593
    ConstraintWeights_4:float=0.1052
    ConstraintWeights_5:float=0.1017
    ConstraintWeights_6:float=0.0139
    ConstraintWeights_7:float=0.0078
   ) -> (Pose) {
    alphaBoolBlend {blendOption=Linear}
    boneToModify {boneName=base_hip_strap1_l}
    constraintSetup [
     {targetBone {boneName=dyn_jacket_0_m}}
     {targetBone {boneName=dyn_jacket_2_l}}
     {targetBone {boneName=dyn_jacket_0_end_m}}
     {targetBone {boneName=dyn_jacket_2_end_l}}
     {targetBone {boneName=dyn_jacket_2_r}}
     {targetBone {boneName=dyn_jacket_2_end_r}}
     {targetBone {boneName=spine_04}}
     {targetBone {boneName=dyn_jacket_1_end_r}}
    ]
    constraintWeights [1,1,1,1,1,1,1,1]
   }
   Save cached pose 'SurfConsAftrCache'(Pose=Component To Local(ComponentPose=@n42)) {cacheName=SurfConsAftrCache}
   n43 = "Transform (Modify) Bone - Bone: dyn_hip_strap3_l"!(
    ComponentPose=Local To Component(LocalPose=Use cached pose 'SurfConsAftrCache'())
    bAlphaBoolEnabled:bool=True
    Translation="0,0,0"
    Rotation="-10,0,0"
   ) -> (Pose) {
    alphaInputType=Bool
    rotationMode=BMM_Additive
    rotationSpace=BCS_BoneSpace
    alphaBoolBlend {blendInTime=1, blendOption=Linear, blendOutTime=1}
    boneToModify {boneName=dyn_hip_strap3_l}
    rotation {pitch=-10}
   }
   n44 = "Transform (Modify) Bone - Bone: dyn_hip_strap3_r"!(ComponentPose=@n43, bAlphaBoolEnabled:bool=True, Translation="0,0,0", Rotation="-10,0,0") -> (Pose) {
    alphaInputType=Bool
    rotationMode=BMM_Additive
    rotationSpace=BCS_BoneSpace
    alphaBoolBlend {blendInTime=1, blendOption=Linear, blendOutTime=1}
    boneToModify {boneName=dyn_hip_strap3_r}
    rotation {pitch=-10}
   }
   n45 = Anim Dynamics - Bone: dyn_hip_strap1_l!(ComponentPose=@n44, bAlphaBoolEnabled:bool=True, GravityOverride:Vector="0,0,-980", ExternalForce="0,0,0") -> (Pose) {
    alphaInputType=Bool
    angularDampingOverride=1
    angularSpringConstant=20000
    bAngularSpring=true
    bOverrideAngularDamping=true
    bUseGravityOverride=true
    gravityScale=0.35
    simulationSpace=RootRelative
    alphaBoolBlend {blendOption=Linear}
    boundBone {boneName=dyn_hip_strap1_l}
    componentAppliedLinearAccClamp=<3000 3000 3000>
    componentLinearAccScale=<0.8 0.8 0.8>
    componentLinearVelScale=<0.8 0.8 0.8>
    gravityOverride {z=-980}
    physicsBodyDefinitions [{
      boundBone {boneName=dyn_hip_strap1_l}
      boxExtents=<30 30 35>
      constraintSetup {angularLimitsMax=<0 0 40>, angularLimitsMin=<0 -6 -5>, angularTarget=<1 0 0>}
      localJointOffset=<5 0 0>
     }]
   }
   n46 = Anim Dynamics - Bone: dyn_hip_strap2_l!(ComponentPose=@n45, bAlphaBoolEnabled:bool=True, GravityOverride:Vector="0,0,-980", ExternalForce="0,0,0") -> (Pose) {
    alphaInputType=Bool
    angularDampingOverride=1
    angularSpringConstant=20000
    bAngularSpring=true
    bOverrideAngularDamping=true
    bUseGravityOverride=true
    gravityScale=0.35
    simulationSpace=RootRelative
    alphaBoolBlend {blendOption=Linear}
    boundBone {boneName=dyn_hip_strap2_l}
    componentAppliedLinearAccClamp=<3000 3000 3000>
    componentLinearAccScale=<0.8 0.8 0.8>
    componentLinearVelScale=<0.8 0.8 0.8>
    gravityOverride {z=-980}
    physicsBodyDefinitions [{
      boundBone {boneName=dyn_hip_strap2_l}
      boxExtents=<30 30 35>
      constraintSetup {angularLimitsMax=<0 0 30>, angularLimitsMin=<0 -15 -5>, angularTarget=<1 0 0>}
      localJointOffset=<5 0 0>
     }]
   }
   n47 = Anim Dynamics - Bone: dyn_hip_strap3_l!(ComponentPose=@n46, bAlphaBoolEnabled:bool=True, GravityOverride:Vector="0,0,-980", ExternalForce="0,0,0") -> (Pose) {
    alphaInputType=Bool
    angularDampingOverride=1
    angularSpringConstant=20000
    bAngularSpring=true
    bOverrideAngularDamping=true
    bUseGravityOverride=true
    gravityScale=0.35
    simulationSpace=RootRelative
    alphaBoolBlend {blendOption=Linear}
    boundBone {boneName=dyn_hip_strap3_l}
    componentAppliedLinearAccClamp=<3000 3000 3000>
    componentLinearAccScale=<0.8 0.8 0.8>
    componentLinearVelScale=<0.8 0.8 0.8>
    gravityOverride {z=-980}
    physicsBodyDefinitions [{
      boundBone {boneName=dyn_hip_strap3_l}
      boxExtents=<30 30 35>
      constraintSetup {angularLimitsMax=<0 0 20>, angularLimitsMin=<0 -10 -5>, angularTarget=<1 0 0>}
      localJointOffset=<5 0 0>
     }]
   }
   n48 = Anim Dynamics - Bone: dyn_hip_strap4_l!(ComponentPose=@n47, bAlphaBoolEnabled:bool=True, GravityOverride:Vector="0,0,-980", ExternalForce="0,0,0") -> (Pose) {
    alphaInputType=Bool
    angularDampingOverride=1
    angularSpringConstant=20000
    bAngularSpring=true
    bOverrideAngularDamping=true
    bUseGravityOverride=true
    gravityScale=0.35
    simulationSpace=RootRelative
    alphaBoolBlend {blendOption=Linear}
    boundBone {boneName=dyn_hip_strap4_l}
    componentAppliedLinearAccClamp=<3000 3000 3000>
    componentLinearAccScale=<0.8 0.8 0.8>
    componentLinearVelScale=<0.8 0.8 0.8>
    gravityOverride {z=-980}
    physicsBodyDefinitions [{
      boundBone {boneName=dyn_hip_strap4_l}
      boxExtents=<30 30 35>
      constraintSetup {angularLimitsMax=<0 0 10>, angularLimitsMin=<0 -5 -5>, angularTarget=<1 0 0>}
      localJointOffset=<5 0 0>
     }]
   }
   n49 = Anim Dynamics - Bone: dyn_hip_strap1_r!(ComponentPose=@n48, bAlphaBoolEnabled:bool=True, GravityOverride:Vector="0,0,-980", ExternalForce="0,0,0") -> (Pose) {
    alphaInputType=Bool
    angularDampingOverride=1
    angularSpringConstant=20000
    bAngularSpring=true
    bOverrideAngularDamping=true
    bUseGravityOverride=true
    gravityScale=0.35
    simulationSpace=RootRelative
    alphaBoolBlend {blendOption=Linear}
    boundBone {boneName=dyn_hip_strap1_r}
    componentAppliedLinearAccClamp=<3000 3000 3000>
    componentLinearAccScale=<0.8 0.8 0.8>
    componentLinearVelScale=<0.8 0.8 0.8>
    gravityOverride {z=-980}
    physicsBodyDefinitions [{
      boundBone {boneName=dyn_hip_strap1_r}
      boxExtents=<30 30 35>
      constraintSetup {angularLimitsMax=<0 0 5>, angularLimitsMin=<0 -4 -40>, angularTarget=<1 0 0>}
      localJointOffset=<5 0 0>
     }]
   }
   n50 = Anim Dynamics - Bone: dyn_hip_strap2_r!(ComponentPose=@n49, bAlphaBoolEnabled:bool=True, GravityOverride:Vector="0,0,-980", ExternalForce="0,0,0") -> (Pose) {
    alphaInputType=Bool
    angularDampingOverride=1
    angularSpringConstant=20000
    bAngularSpring=true
    bOverrideAngularDamping=true
    bUseGravityOverride=true
    gravityScale=0.35
    simulationSpace=RootRelative
    alphaBoolBlend {blendOption=Linear}
    boundBone {boneName=dyn_hip_strap2_r}
    componentAppliedLinearAccClamp=<3000 3000 3000>
    componentLinearAccScale=<0.8 0.8 0.8>
    componentLinearVelScale=<0.8 0.8 0.8>
    gravityOverride {z=-980}
    physicsBodyDefinitions [{
      boundBone {boneName=dyn_hip_strap2_r}
      boxExtents=<30 30 35>
      constraintSetup {angularLimitsMax=<0 0 5>, angularLimitsMin=<0 -20 -35>, angularTarget=<1 0 0>}
      localJointOffset=<5 0 0>
     }]
   }
   n51 = Anim Dynamics - Bone: dyn_hip_strap3_r!(ComponentPose=@n50, bAlphaBoolEnabled:bool=True, GravityOverride:Vector="0,0,-980", ExternalForce="0,0,0") -> (Pose) {
    alphaInputType=Bool
    angularDampingOverride=1
    angularSpringConstant=20000
    bAngularSpring=true
    bOverrideAngularDamping=true
    bUseGravityOverride=true
    gravityScale=0.35
    simulationSpace=RootRelative
    alphaBoolBlend {blendOption=Linear}
    boundBone {boneName=dyn_hip_strap3_r}
    componentAppliedLinearAccClamp=<3000 3000 3000>
    componentLinearAccScale=<0.8 0.8 0.8>
    componentLinearVelScale=<0.8 0.8 0.8>
    gravityOverride {z=-980}
    physicsBodyDefinitions [{
      boundBone {boneName=dyn_hip_strap3_r}
      boxExtents=<30 30 35>
      constraintSetup {angularLimitsMax=<0 0 5>, angularLimitsMin=<0 -15 -25>, angularTarget=<1 0 0>}
      localJointOffset=<5 0 0>
     }]
   }
   n52 = Anim Dynamics - Bone: dyn_hip_strap4_r!(ComponentPose=@n51, bAlphaBoolEnabled:bool=True, GravityOverride:Vector="0,0,-980", ExternalForce="0,0,0") -> (Pose) {
    alphaInputType=Bool
    angularDampingOverride=1
    angularSpringConstant=20000
    bAngularSpring=true
    bOverrideAngularDamping=true
    bUseGravityOverride=true
    gravityScale=0.35
    simulationSpace=RootRelative
    alphaBoolBlend {blendOption=Linear}
    boundBone {boneName=dyn_hip_strap4_r}
    componentAppliedLinearAccClamp=<3000 3000 3000>
    componentLinearAccScale=<0.8 0.8 0.8>
    componentLinearVelScale=<0.8 0.8 0.8>
    gravityOverride {z=-980}
    physicsBodyDefinitions [{
      boundBone {boneName=dyn_hip_strap4_r}
      boxExtents=<30 30 35>
      constraintSetup {angularLimitsMax=<0 0 5>, angularLimitsMin=<0 -8 -20>, angularTarget=<1 0 0>}
      localJointOffset=<5 0 0>
     }]
   }
   Save cached pose 'HipStrapsCache'(Pose=Component To Local(ComponentPose=@n52)) {cacheName=HipStrapsCache}
   Save cached pose 'Inputpose'(Pose=Use cached pose 'Start_Cache'()) {cacheName=Inputpose}
  ]}
 EventGraph {nodes [
   Event OnCharacterCustomizationLoaded()
   > n1 = Cast To PlayerPawn_Athena(Object=TryGetPawnOwner())
   > Set Body_Mesh(Body_Mesh=GetSkeletalMeshForPartType(self=@n1))
   n2 = CheckForBackbling()
   > Set isWearingBackbling(isWearingBackbling=AND Boolean!(A=@n2, B=NOT Boolean(A=@n2.bIsBackblingACape), C:bool=NOT Boolean(A=@n2.bIsBackblingLong)) -> (ReturnValue))
   > n3 = Set Is Wearing Cape(Is Wearing Cape=OR Boolean(A=@n2.bIsBackblingACape, B=@n2.bIsBackblingLong))
   > n4 = Branch(Condition=@n3)
   > Set CapeVar(CapeVar:bool=false) -> (Output_Get:bool)
   Set CapeVar(execute=@n4.else, CapeVar=true)
   Event BlueprintUpdateAnimation()
   > n5 = Sequence!(execute) -> (then_0, then_1, then_2:Exec, then_3:Exec)
   > n6 = Gravity Override State Function DEPRECATED(CharacterPartAnimInstance=Self-Reference(), GameTimeInSeconds=GetGameTimeInSeconds(), GravityOverrideParams=Get Gravity Override Params())
   > Set GravityOverride(GravityOverride=@n6)
   > n7 = Branch(Condition=Get bIsSprinting())
   > Set isSprinting(isSprinting:double=1) -> (Output_Get:double)
   Set isSprinting(execute=@n7.else, isSprinting=1)
   n8 = PlayerMovement(execute=@n5.then_1, AnimInstance=Self-Reference(), XDotProductCoefficient=100, YDotProductCoefficient=100, ZDotProductCoefficient=100)
   n9 = BreakVector(InVec=MakeVector(X=@n8, Y=@n8.YMovement))
   n10 = "float < float"(A=@n9, B=-50)
   n11 = "float > float"(A=@n9, B=50)
   n12 = Get bIsCrouching()
   > n13 = Set CrouchingSides(CrouchingSides:double=float * float(A:bool=OR Boolean(A=@n10, B=@n11), B:bool=@n12) {is=title K2Node_PromotableOperator}) -> (Output_Get:double)
   > Set CrouchingSidesRB(CrouchingSidesRB:double=Remap Float(Input=@n13, OldRange=(X=0,Y=1), NewRange=(X=1,Y=0.5), Clamp=true) -> (Output=~)) -> (Output_Get:double)
   > Set RightSide(RightSide:double=float * float(A=@n11, B=@n12)) -> (Output_Get:double)
   n14 = NOT Boolean(A=OR Boolean(A=Get bIsCrouching(), B=Get bIsJumping()))
   n15 = Get bIsMoving2D()
   > Set moveToRight(moveToRight:double=float * float(A=@n11, B=AND Boolean!(A=@n15, B=@n14) -> (ReturnValue))) -> (Output_Get:double)
   > Set jumpToRight(jumpToRight:double=float * float(A=@n11, B=Get bIsJumping())) -> (Output_Get:double)
   n16 = GetOwnerSkeletalMeshComponent()
   n17 = BreakRotator(InRot=BreakTransform(InTransform=GetDeltaTransformFromRefPose(self=@n16, BoneName=thigh_r, BaseName=pelvis)).Rotation)
   Set jump(execute=@n5.then_2, jump:double=float * float(A:double=MapRangeClamped(Value=@n17.Yaw, InRangeA=0, InRangeB=-45), B=Get bIsJumping())) -> (Output_Get:double)
   Set targetStandingUp(execute=@n5.then_3, targetStandingUp:bool=NOT Boolean(A=Get bIsCrouching())) -> (Output_Get:bool)
   n18 = Get bIsSkyDiving()
   > Set frontend(frontend:bool=AND Boolean(A=Get bIsInFrontEnd(), B=NOT Boolean(A=@n18))) -> (Output_Get:bool)
   K2Node_Event_3 = CustomEvent!() -> (
    OutputDelegate
    then
    CustomBaseMesh:Skeletal Mesh Component Object Reference
    CustomParts:Array of Custom Character Part Object References
    CustomMeshParts:Array of Skeletal Mesh Component Budgeted Object References
   ) {is=title K2Node_CustomEvent}
   > n19 = Cast To PlayerPawn_Athena(Object=TryGetPawnOwner())
   > "Set Body Mesh "("Body Mesh ":Skeletal Mesh Component Object Reference=GetSkeletalMeshForPartType(self=@n19)) -> (Output_Get:Skeletal Mesh Component Object Reference)
   > "Set Face Mesh "("Face Mesh ":Skeletal Mesh Component Object Reference=GetSkeletalMeshForPartType(self=@n19, PartType=Head)) -> (Output_Get:Skeletal Mesh Component Object Reference)
  ]}
 NewEventDispatcher {nodes [NewEventDispatcher() {is=title K2Node_FunctionEntry}]}
]
variables [
 "Body Mesh " {type=SkeletalMeshComponent}
 Body_Mesh {type=SkeletalMeshComponent}
 CapeVar {type=boolean}
 CrouchingSides {type=number}
 CrouchingSidesRB {type=number, default=1}
 "Face Mesh " {type=SkeletalMeshComponent}
 frontend {type=boolean}
 Gravity Override Params {type=GravityOverrideParamsStruct}
 GravityOverride {type=Vector}
 Is Wearing Cape {type=boolean}
 isSprinting {type=number}
 isWearingBackbling {type=boolean}
 jump {type=number}
 jumpToRight {type=number}
 moveToRight {type=number}
 NewEventDispatcher {default=()}
 RightSide {type=number}
 Skel Mesh {type=SkeletalMeshComponent}
 targetStandingUp {type=boolean}
]
