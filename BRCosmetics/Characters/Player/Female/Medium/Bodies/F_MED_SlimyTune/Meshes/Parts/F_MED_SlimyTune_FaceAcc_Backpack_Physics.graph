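# Physics asset for the FaceAcc/Backpack part of the SlimyTune female medium character, giving its hair bones physical shapes and joints.
# A main chain runs from base_hair_0 through dyn_hair_0 to dyn_hair_4, with short branches off dyn_hair_0 and dyn_hair_1.
# Every joint has narrow swing limits and zero twist limit.
# Unconstrained capsules on neck, upper spine and upper arms approximate the body, apparently as collision for the dynamic bones.
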PhysicsAsset /BRCosmetics/Characters/Player/Female/Medium/Bodies/F_MED_SlimyTune/Meshes/Parts/F_MED_SlimyTune_FaceAcc_Backpack_Physics
bodies {
 base_hair_0 [{shapeType=Capsule, length=0.505, radius=0.505, shapeName=base_hair_0_capsule, rotation=<rot 0 -90 0>}]
 dyn_hair_0 [{shapeType=Capsule, length=5, radius=2, shapeName=dyn_hair_0_capsule, center=<2 0 0>, rotation=<rot 90 0 0>}]
 dyn_hair_1 [{
   shapeType=Capsule
   length=6.5268
   radius=6.269
   shapeName=dyn_hair_1_capsule
   center=<3.6686 -0.6978 2.1631>
   rotation=<rot 90 175.2364 90.1428>
  }]
 dyn_hair_2 [{shapeType=Capsule, length=4.6412, radius=5.0882, shapeName=dyn_hair_2_capsule, center=<2 0 1.3545>, rotation=<rot 90 0 0>}]
 dyn_hair_3 [{shapeType=Capsule, length=3.3354, radius=4.5707, shapeName=dyn_hair_3_capsule, center=<2 0 0.6459>, rotation=<rot 90 0 0>}]
 dyn_hair_4 [{
   shapeType=Capsule
   length=8.161
   radius=5.9099
   shapeName=dyn_hair_4_capsule
   center=<1.8593 -2.8342 4.7732>
   rotation=<rot 56.6322 164.9482 -156.9801>
  }]
 dyn_hair_5 [{shapeType=Capsule, length=6.9436, radius=1.4, shapeName=dyn_hair_5_capsule, center=<2 0 0>, rotation=<rot 90 0 0>}]
 dyn_hair_6 [{
   shapeType=Capsule
   length=6.6776
   radius=1.6591
   shapeName=dyn_hair_6_capsule
   center=<5.5854 0 -0.3459>
   rotation=<rot 76.7285 -4.1184 -14.767>
  }]
 dyn_hair_7 [{shapeType=Capsule, length=5.0346, radius=1.1609, shapeName=dyn_hair_7_capsule, center=<2 0 0>, rotation=<rot 90 0 0>}]
 neck_01 [{shapeType=Capsule, length=11.294, radius=8.706, shapeName=neck_01_capsule, center=<3.4109 -0.5025 0>}]
 spine_04 [{shapeType=Capsule, length=13.0808, radius=15.9827, shapeName=spine_05_capsule, center=<-15.4346 1.2974 0>, rotation=<rot -87.1897 7.4673 0>}]
 spine_05 [
  {shapeType=Capsule, length=12.2907, radius=14.196, shapeName=spine_05_capsule, center=<-1.997 2.2 0>}
  {shapeType=Capsule, length=31.3048, radius=15, shapeName=spine_05_capsule, center=<0 -38.711 0>}
  {shapeType=Capsule, length=25.6462, radius=15, shapeName=spine_05_capsule, center=<25.4042 -38.7501 0>}
 ]
 upperarm_l [{shapeType=Capsule, length=35.406, radius=7.876, shapeName=upperarm_l_capsule, center=<16.5169 0 0>, rotation=<rot 90 0 0>}]
 upperarm_r [{shapeType=Capsule, length=35.406, radius=7.876, shapeName=upperarm_r_capsule, center=<-13.7702 0 0>, rotation=<rot 90 0 0>}]
}
constraints [
 {bone1Name=dyn_hair_0, bone2Name=base_hair_0, swing1LimitDegrees=10, swing2LimitDegrees=20, twistLimitDegrees=0}
 {bone1Name=dyn_hair_1, bone2Name=dyn_hair_0, swing1LimitDegrees=5, swing2LimitDegrees=5, twistLimitDegrees=0}
 {bone1Name=dyn_hair_2, bone2Name=dyn_hair_1, swing1LimitDegrees=5, swing2LimitDegrees=5, twistLimitDegrees=0}
 {bone1Name=dyn_hair_3, bone2Name=dyn_hair_2, swing1LimitDegrees=5, swing2LimitDegrees=5, twistLimitDegrees=0}
 {bone1Name=dyn_hair_4, bone2Name=dyn_hair_3, swing1LimitDegrees=5, swing2LimitDegrees=5, twistLimitDegrees=0}
 {bone1Name=dyn_hair_5, bone2Name=dyn_hair_0, swing1LimitDegrees=10, swing2LimitDegrees=10, twistLimitDegrees=0}
 {bone1Name=dyn_hair_7, bone2Name=dyn_hair_1, swing1LimitDegrees=20, swing2LimitDegrees=20, twistLimitDegrees=0}
 {bone1Name=dyn_hair_6, bone2Name=dyn_hair_1, swing1LimitDegrees=10, swing2LimitDegrees=10, twistLimitDegrees=0}
]
