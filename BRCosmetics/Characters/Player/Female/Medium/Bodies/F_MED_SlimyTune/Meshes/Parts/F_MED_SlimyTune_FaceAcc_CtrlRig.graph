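# Control Rig for the F_MED_SlimyTune face accessory mesh that adjusts the base of the character's hair chain.
# A local-space-offset aim constraint turns base_hair_0 toward aim_hair, which hangs under spine_05, carrying the dyn_hair children along.
# The head_up_crouching variable sets the constraint weight and defaults to zero.
# A head_ctrl-to-head transform copy is present but disconnected from execution.
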ControlRigBlueprint /BRCosmetics/Characters/Player/Female/Medium/Bodies/F_MED_SlimyTune/Meshes/Parts/F_MED_SlimyTune_FaceAcc_CtrlRig
graphs [
 RigVMFunctionLibrary {links [], nodes []}
 RigVMModel {nodes [
   RigUnit_BeginExecution = &RigUnit_BeginExecution()
   &RigUnit_AimConstraintLocalSpaceOffset(
    Child=(Type=Bone,Name="base_hair_0")
    Filter=(bX=True,bY=True,bZ=True)
    Parents=((Item=(Type=Bone,Name="aim_hair"),Weight=1))
    UpAxis=(X=0,Y=1,Z=0)
    WorldUp=(Target=(X=0,Y=0,Z=0),Kind=Direction,Space=(Type=None,Name="None"))
    ExecutePin=@RigUnit_BeginExecution.ExecutePin
    Weight=RigVMVariableNode() {variable=head_up_crouching}.Value
   )
   RigUnit_GetTransform = &RigUnit_GetTransform(Item=(Type=Control,Name="head_ctrl"), bInitial=False)
   RigUnit_SetTransform = &RigUnit_SetTransform(ExecutePin=(), Item=(Type=Bone,Name="head"), Weight=1, bInitial=False, Value=@RigUnit_GetTransform.Transform)
  ]}
]
hierarchy [
 root {loc=<0 0 0>, rot=<0 0 0 1>}
 pelvis {parent=root, loc=<0 -0.5301 98.0571>, rot=<0.0127 0.707 -0.0127 -0.707>}
 spine_01 {parent=pelvis, loc=<6.1208 -0.9328 0>, rot=<0 0 0.034 -0.9994>}
 spine_02 {parent=spine_01, loc=<6.5997 -0.2095 0>, rot=<0 0 0.0238 -0.9997>}
 spine_03 {parent=spine_02, loc=<6.4914 -0.1237 0>, rot=<0 0 0.0149 -0.9999>}
 spine_04 {parent=spine_03, loc=<6.9598 0.0794 0>, rot=<0 0 -0.0132 -0.9999>}
 spine_05 {parent=spine_04, loc=<6.9987 0 0>, rot=<0 0 -0.0015 -1>}
 neck_01 {parent=spine_05, loc=<11.3138 0.1529 0>, rot=<0 0 -0.2003 -0.9797>}
 neck_02 {parent=neck_01, loc=<4.5627 0 0>, rot=<0 0 0.0504 -0.9987>}
 head {parent=neck_02, loc=<5.3914 0 0>, rot=<0 0 0.115 -0.9934>}
 base_hair_0 {parent=head, loc=<6.0051 -9.0962 0.0002>, rot=<0.5013 -0.4987 0.4987 -0.5013>}
 dyn_hair_0 {parent=base_hair_0, loc=<0 0 0>, rot=<0 0 0 1>}
 dyn_hair_1 {parent=dyn_hair_0, loc=<8.2287 -0.0092 0.103>, rot=<-0.0594 -0.7463 -0.0526 -0.6609>}
 dyn_hair_2 {parent=dyn_hair_1, loc=<9.2502 0 0>, rot=<-0.0006 0.0228 0.0249 -0.9994>}
 dyn_hair_3 {parent=dyn_hair_2, loc=<8.0095 0 0>, rot=<-0.008 0.0897 0.0888 -0.992>}
 dyn_hair_4 {parent=dyn_hair_3, loc=<8.6849 0 0>, rot=<0 0 0 1>}
 clavicle_l {parent=spine_05, loc=<5.4831 -0.1053 -3.2946>, rot=<0.0465 -0.7289 -0.009 -0.6829>}
 upperarm_l {parent=clavicle_l, loc=<9.8658 -1.2136 -0.4105>, rot=<0.0164 -0.3245 -0.0269 -0.9454>}
 clavicle_r {parent=spine_05, loc=<5.4831 -0.1053 3.2946>, rot=<0.7289 0.0465 -0.6829 0.009>}
 upperarm_r {parent=clavicle_r, loc=<-9.8658 1.2136 0.4105>, rot=<0.0164 -0.3245 -0.0269 -0.9454>}
 Z_aim_hair {parent=spine_05, loc=<1.8481 -11.1302 -2.3091>, rot=<0.0284 -0.7065 -0.0284 -0.7065>}
 aim_hair {parent=Z_aim_hair, loc=<0 0 0>, rot=<0 0 0 1>}
 head_ctrl {controlType=EULER_TRANSFORM, type=Control, loc=<0 0 0>, rot=<0 0 0 1>}
]
variables [head_up_crouching {type=double, default=0}]
